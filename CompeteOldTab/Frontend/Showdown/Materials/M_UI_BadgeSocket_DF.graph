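# Translucent UI material that draws a Showdown badge socket from a distance-field texture, with fill, outline and shadow.
# AAThreshold cuts the red channel at Edge and Edge plus OutlineSize to layer shadow, outline and fill colors.
# Ended and Focus parameters blend each layer's color, with Focus applied last.
# The blue channel, faded by a rotated gradient, forms shadow opacity, which Ended removes, then maxes with the shape mask.
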Material /CompeteOldTab/Frontend/Showdown/Materials/M_UI_BadgeSocket_DF
graph {
 properties {blendMode=BLEND_Translucent, materialDomain=MD_UI, pixelDepthOffsetMode=PDOM_Legacy, usedShadingModels=Default Lit}
 nodes [
  n1 = VectorParameter() {parameterName=ShadowColor, defaultValue=<rgba 0.007 0.2874 0.5029 0.43>}
  n2 = ScalarParameter() {parameterName=Ended}
  n3 = LinearInterpolate(A=@n1.RGB, B=VectorParameter() {parameterName=ShadowColor_Ended, defaultValue=<rgba 0 0 0 1>}.RGB, Alpha=@n2)
  n4 = VectorParameter() {parameterName=ShadowColor_Focus, defaultValue=<rgba 0.007 0.2874 0.5029 0.43>}
  n5 = ScalarParameter() {parameterName=Focus}
  n6 = LinearInterpolate(
   A=VectorParameter() {parameterName=OutlineColor, defaultValue=<rgba 0.2016 0.9823 1 1>}.RGB
   B=VectorParameter() {parameterName=OutlineColor_Ended, defaultValue=<rgba 0 0 0 1>}.RGB
   Alpha=ScalarParameter() {parameterName=Ended}
  )
  n7 = VectorParameter() {parameterName=OutlineColor_Focus, defaultValue=<rgba 0.2016 0.9823 1 1>}
  n8 = ScalarParameter() {parameterName=Focus}
  n9 = ScalarParameter() {defaultValue=0.1, parameterName=Edge}
  n10 = TextureCoordinate()
  n11 = ScalarParameter() {defaultValue=1, parameterName=Scale}
  n12 = VectorParameter() {parameterName=Position, defaultValue=<rgba 0 0 0 1>}
  n13 = TextureSample(UVs=Add(
    A=MaterialFunctionCall(UVs=@n10, Texture Scale=@n11) {materialFunction=&ScaleUVsByCenter}.UVs
    B=AppendVector(A=@n12.R, B=@n12.R)
   )) {texture=&/T_UI_BadgeSocket_DF}
  n14 = ScalarParameter() {defaultValue=0.01, parameterName=EdgeSoftness}
  n15 = MaterialFunctionCall!(Threshold=@n9, Gradient=@n13.R, Filter Width Override=@n14, Filter Width Fudge Factor) {materialFunction=/Game/UI/Materials/AAThreshold}
  n16 = VectorParameter() {parameterName=ShapeColor, defaultValue=<rgba 0.0009 0.8632 0.8388 1>}
  n17 = VectorParameter() {parameterName=ShapeColor_Ended, defaultValue=<rgba 0 0 0 1>}
  n18 = ScalarParameter() {parameterName=Ended}
  n19 = ScalarParameter() {defaultValue=0.1, parameterName=OutlineSize}
  LinearInterpolate(
   A=LinearInterpolate(A=LinearInterpolate(A=@n3, B=@n4.RGB, Alpha=@n5), B=LinearInterpolate(A=@n6, B=@n7.RGB, Alpha=@n8), Alpha=@n15.Result)
   B=LinearInterpolate(
    A=LinearInterpolate(A=@n16.RGB, B=@n17.RGB, Alpha=@n18)
    B=VectorParameter() {parameterName=ShapeColor_Focus, defaultValue=<rgba 0 0.15 1 1>}.RGB
    Alpha=ScalarParameter() {parameterName=Focus}
   )
   Alpha=MaterialFunctionCall(Threshold=Add(A=@n9, B=@n19), Gradient=@n13.R, Filter Width Override=@n14) {materialFunction=&AAThreshold}.Result
  )
  n20 = TextureCoordinate()
  n21 = ScalarParameter() {defaultValue=0.1, parameterName=ShadowFadeDirection}
  n22 = VectorParameter() {parameterName=ShadowFade}
  n23 = Add(
   A=ComponentMask(None=MaterialFunctionCall(UVs=@n20, "Rotation Angle (0-1)"=@n21) {materialFunction=&CustomRotator}.Rotated Values) {g=true}
   B=AppendVector(A=@n22.R, B=@n22.R)
  )
  n24 = Multiply(A=@n13.B, B=OneMinus(None=Multiply(A=@n23, B=@n22.A)))
  n25 = OneMinus(None=@n2)
  Max(
   A=@n15.Result
   B=Multiply(
    A=MaterialFunctionCall(Threshold=@n9, Gradient=@n24, Filter Width Override=@n14) {materialFunction=&AAThreshold}.Result
    B=Multiply(A=@n25, B=@n1.A)
   )
  )
 ]
}
