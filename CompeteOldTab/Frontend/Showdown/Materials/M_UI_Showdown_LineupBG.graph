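# M_UI_Showdown_LineupBG is a translucent UI material for a Showdown lineup background, combining an angled color band with two glows.
# Two rotatable smoothstep edges cut the band, separating AngleColor from CornerGlow, with Glow2 blended over by the second radial gradient.
# Both exponential gradients use aspect-corrected screen-aligned UVs, with centers using only the position parameters' R channel.
# The combined band and gradient masks fade the color in; the Opacity parameter is unconnected.
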Material /CompeteOldTab/Frontend/Showdown/Materials/M_UI_Showdown_LineupBG
graph {
 properties {blendMode=BLEND_Translucent, materialDomain=MD_UI, pixelDepthOffsetMode=PDOM_Legacy, usedShadingModels=Default Lit}
 nodes [
  n1 = VectorParameter() {group=Radial, parameterName=CornerGlow, defaultValue=<rgba 0.0012 0.7157 0.9301 1>}
  n2 = VectorParameter() {group=Angle, parameterName=AngleColor, defaultValue=<rgba 0.0015 0.6939 0.9047 1>}
  n3 = TextureCoordinate()
  n4 = ScalarParameter() {defaultValue=0.1, group=Angle, parameterName=LeftEdge}
  n5 = AppendVector(A=@n4, B=Constant() {r=0.5})
  n6 = ScalarParameter() {group=Angle, parameterName=LeftRotation, sliderMax=0.25, sliderMin=-0.25}
  n7 = MaterialFunctionCall(
   Alpha=ComponentMask(None=MaterialFunctionCall(UVs=@n3, Rotation Center=@n5, "Rotation Angle (0-1)"=@n6) {materialFunction=&CustomRotator}.Rotated Values) {r=true}
   Min=@n4
   Max=Add(A=@n4, B=ScalarParameter() {defaultValue=0.001, group=Angle, parameterName=EdgeSoftness})
  ) {materialFunction=&SmoothStep}
  n8 = OneMinus(None=ScalarParameter() {defaultValue=0.1, group=Angle, parameterName=RightEdge})
  n9 = Constant() {r=0.5}
  n10 = ComponentMask(None=MaterialFunctionCall(
    UVs=TextureCoordinate()
    Rotation Center=AppendVector(A=@n8, B=@n9)
    "Rotation Angle (0-1)"=ScalarParameter() {group=Angle, parameterName=RightRotation, sliderMax=0.25, sliderMin=-0.25}
   ) {materialFunction=&CustomRotator}.Rotated Values) {r=true}
  n11 = Add(A=@n8, B=ScalarParameter() {defaultValue=0.001, group=Angle, parameterName=EdgeSoftness})
  n12 = OneMinus(None=Max(
    A=OneMinus(None=@n7.Result)
    B=MaterialFunctionCall(Alpha=@n10, Min=@n8, Max=@n11) {materialFunction=&SmoothStep}.Result
   ))
  n13 = MaterialFunctionCall(Texture Dimensions=Constant2Vector() {g=512, r=512}) {materialFunction=&ScreenAlignedUVs}
  n14 = VectorParameter() {group=Radial2, parameterName=RadialPosition_2, defaultValue=<rgba -1 1 0 1>}
  n15 = MaterialFunctionCall(
   UVs=@n13."X 100%, Y Scale to Ratio"
   CenterPosition=AppendVector(A=@n14.R, B=@n14.R)
   Radius=ScalarParameter() {defaultValue=2.5, group=Radial2, parameterName=RadialSize2}
   Density=ScalarParameter() {defaultValue=2.3, group=Radial2, parameterName=Density2}
  ) {materialFunction=&RadialGradientExponential}
  n16 = VectorParameter() {group=Radial, parameterName=RadialPosition, defaultValue=<rgba 2 0 0 1>}
  n17 = MaterialFunctionCall(
   UVs=@n13."X 100%, Y Scale to Ratio"
   CenterPosition=AppendVector(A=@n16.R, B=@n16.R)
   Radius=ScalarParameter() {defaultValue=2.5, group=Radial, parameterName=RadialSize}
   Density=ScalarParameter() {defaultValue=2.3, group=Radial, parameterName=Density}
  ) {materialFunction=&RadialGradientExponential}
  n18 = Multiply(A=@n12, B=@n2.A)
  LinearInterpolate(
   B=LinearInterpolate(
    A=LinearInterpolate(A=@n1.RGB, B=@n2.RGB, Alpha=@n12)
    B=VectorParameter() {group=Radial2, parameterName=Glow2, defaultValue=<rgba 0.0075 0.0091 0.0242 1>}.RGB
    Alpha=@n15.RadialGradientExponential
   )
   Alpha=Max(A=@n15.RadialGradientExponential, B=Max(A=@n17.RadialGradientExponential, B=@n18))
  )
  MaterialFunctionCall() {materialFunction=&ScreenAlignedUVs}
  ScalarParameter() {defaultValue=0.0905, parameterName=Opacity}
 ]
}
