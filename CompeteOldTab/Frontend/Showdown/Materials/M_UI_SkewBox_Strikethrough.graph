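# Translucent UI material for the Showdown frontend that computes a thin, rotatable horizontal band mask, named as a skew-box strikethrough.
# Two CustomRotator UV branches share Rotation, each pivoting on its edge value paired with 0.5.
# SmoothStep on each rotated V softens edges derived from BottomEdge, with Width offsetting one of them.
# Max merges both outside regions, and a final OneMinus leaves the band between them.
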Material /CompeteOldTab/Frontend/Showdown/Materials/M_UI_SkewBox_Strikethrough
graph {
 properties {blendMode=BLEND_Translucent, materialDomain=MD_UI, pixelDepthOffsetMode=PDOM_Legacy, usedShadingModels=Default Lit}
 nodes [
  Constant() {r=1}
  n1 = ScalarParameter() {defaultValue=0.45, parameterName=BottomEdge}
  n2 = OneMinus(None=@n1)
  n3 = Constant() {r=0.5}
  n4 = ScalarParameter() {parameterName=Rotation, sliderMax=0.25, sliderMin=-0.25}
  n5 = ComponentMask(None=MaterialFunctionCall(UVs=TextureCoordinate(), Rotation Center=AppendVector(A=@n2, B=@n3), "Rotation Angle (0-1)"=@n4) {materialFunction=&CustomRotator}.Rotated Values) {g=true}
  n6 = Add(A=@n2, B=ScalarParameter() {defaultValue=0.001, parameterName=EdgeSoftness})
  n7 = TextureCoordinate()
  n8 = Add(A=@n1, B=ScalarParameter() {defaultValue=0.01, parameterName=Width})
  n9 = AppendVector(A=@n8, B=Constant() {r=0.5})
  n10 = MaterialFunctionCall(
   Alpha=ComponentMask(None=MaterialFunctionCall(UVs=@n7, Rotation Center=@n9, "Rotation Angle (0-1)"=@n4) {materialFunction=&CustomRotator}.Rotated Values) {g=true}
   Min=@n8
   Max=Add(A=@n8, B=ScalarParameter() {defaultValue=0.001, parameterName=EdgeSoftness})
  ) {materialFunction=&SmoothStep}
  OneMinus(None=Max(
    A=MaterialFunctionCall(Alpha=@n5, Min=@n2, Max=@n6) {materialFunction=&SmoothStep}.Result
    B=OneMinus(None=@n10.Result)
   ))
 ]
}
