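# Translucent UI material that, judging by its name, provides the poster fade on the Showdown tournament details screen.
# An exponential radial gradient function takes texture coordinates scaled by the Stretch parameter, plus Center, Radius and Density parameters.
# Stretch and Center each feed only their red channel into both axes, so scaling is uniform and the center lies on the diagonal.
# A blue Color parameter is defined alongside, with no material output connections shown.
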Material /CompeteOldTab/Frontend/Showdown/Materials/M_UI_TournamentDetails_PosterFade
graph {
 properties {blendMode=BLEND_Translucent, materialDomain=MD_UI, pixelDepthOffsetMode=PDOM_Legacy, usedShadingModels=Default Lit}
 nodes [
  n1 = VectorParameter() {parameterName=Stretch, defaultValue=<rgba 1 1 0 1>}
  n2 = VectorParameter() {parameterName=Center, defaultValue=<rgba 1 1 0 1>}
  MaterialFunctionCall(
   UVs=Multiply(A=TextureCoordinate(), B=AppendVector(A=@n1.R, B=@n1.R))
   CenterPosition=AppendVector(A=@n2.R, B=@n2.R)
   Radius=ScalarParameter() {defaultValue=2, parameterName=Radius}
   Density=ScalarParameter() {defaultValue=2, parameterName=Density}
  ) {materialFunction=&RadialGradientExponential}
  VectorParameter() {parameterName=Color, defaultValue=<rgba 0.0003 0.1022 0.6308 1>}
 ]
}
