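# Placeholder tab screen for the Compete frontend, built as a CommonActivatableWidget that hosts a placeholder modal.
# An overlay layers a full-fill radial gradient image behind a centered WBP_CompetePlaceholder_Modal.
# On activation, the tab activates the nested modal; the OnInitialized and Construct events are empty.
# Boolean variables for rank icon, blue, highlight, locked and solo states are declared but unused in this graph.
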WidgetBlueprint /CompeteOldTab/Frontend/Showdown/CompetePlaceholders/WBP_CompetePlaceholderTab
graphs [EventGraph {nodes [
   Event On Activated()
   > ActivateWidget(self=Get WBP_CompetePlaceholder_Modal() -> (WBP_CompetePlaceholder_Modal:WBP Compete Placeholder Modal Object Reference))
   Event OnInitialized()
   Event Construct()
  ]}]
variables [
 HasRankIcon {type=boolean}
 IsBlue {type=boolean}
 IsHighlight {type=boolean}
 IsLocked {type=boolean}
 IsSolo {type=boolean}
]
widgets {
 info {inheritedWidgetCount=0, parentClass=&CommonActivatableWidget, rootWidgetClass=&Overlay, widgetCount=3}
 widgets [
  Overlay_1 {widgetClassPath=&Overlay, widget=Overlay_1}
  Blue_gradient {
   widgetClassPath=&Image
   bIsVariable=true
   parent=Overlay_1
   widget=Blue_gradient
   properties {brush {
     bIsDynamicallyLoaded=false
     drawAs=Image
     imageType=NoImage
     mirroring=NoMirror
     resourceName=None
     resourceObject=/FortUILibrary/Systems/ModularButton/Materials/TestEnvironment/MI_UI_RadialGradient
     tiling=NoTile
     imageSize=<32 32>
     margin=<ltrb 0 0 0 0>
     outlineSettings {
      bUseBrushTransparency=false
      roundingType=HalfHeightRadius
      width=0
      color {colorUseRule=UseColor_Specified, specifiedColor=<rgba 0 0 0 0>}
      cornerRadii=<0 0 0 0>
     }
     tintColor {colorUseRule=UseColor_Specified, specifiedColor=<rgba 1 1 1 1>}
     uVRegion {bIsValid=false, max=<0 0>, min=<0 0>}
    }}
   slot {class=OverlaySlot, properties {horizontalAlignment=HAlign_Fill, verticalAlignment=VAlign_Fill}}
  }
  WBP_CompetePlaceholder_Modal {
   widgetClassPath=&/WBP_CompetePlaceholder_Modal
   bIsVariable=true
   parent=Overlay_1
   widget=WBP_CompetePlaceholder_Modal
   slot {class=OverlaySlot, properties {horizontalAlignment=HAlign_Center, verticalAlignment=VAlign_Center}}
  }
 ]
}
