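# Widget Blueprint for the Showdown tournament poster violator, a bottom-right badge showing an event's status, built on FortTournamentPosterViolator.
# On OnViolatorStateRefreshed, a Select on the state enum picks which border the inherited widget switcher displays.
# Hidden shows an empty border, Live a LiveViolator widget, Countdown the time-remaining text, and Info a width-capped information text.
# The graph never sets either text, and both are inherited bindings from the parent class.
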WidgetBlueprint /CompeteOldTab/Frontend/Showdown/ShowdownPosterViolator
graphs [EventGraph {nodes [
   n1 = Event OnViolatorStateRefreshed!() -> (OutputDelegate, then, NewTimeState:ETournmentPosterViolatorState Enum=Hidden) {is=title K2Node_Event}
   > SetActiveWidget(
    self:Widget Switcher Object Reference=Get WidgetSwitcher_EventStateSwitcher() -> (WidgetSwitcher_EventStateSwitcher:Common Widget Switcher Legacy Object Reference)
    Widget=Select!(
     Hidden:Common Border Object Reference=Get NoViolator() -> (NoViolator:Common Border Object Reference)
     Live:Common Border Object Reference=Get LiveEventViolator() -> (LiveEventViolator:Common Border Object Reference)
     Countdown:Common Border Object Reference=Get FutureEventViolator() -> (FutureEventViolator:Common Border Object Reference)
     Info:Common Border Object Reference=Get InformationViolator() -> (InformationViolator:Common Border Object Reference)
     Index:ETournmentPosterViolatorState Enum=@n1.NewTimeState
    ) -> (ReturnValue:Common Border Object Reference)
   )
  ]}]
variables []
widgets {
 info {inheritedWidgetCount=3, parentClass=/Script/FortniteUI.FortTournamentPosterViolator, rootWidgetClass=&Overlay, widgetCount=10}
 widgets [
  Overlay_0 {widgetClassPath=&Overlay, widget=Overlay_0}
  WidgetSwitcher_EventStateSwitcher {
   widgetClassPath=&CommonWidgetSwitcherLegacy
   bInherited=true
   bIsVariable=true
   parent=Overlay_0
   widget=WidgetSwitcher_EventStateSwitcher
   slot {class=OverlaySlot, properties {horizontalAlignment=HAlign_Right, verticalAlignment=VAlign_Bottom}}
  }
  LiveEventViolator {
   widgetClassPath=&CommonBorder
   bIsVariable=true
   parent=WidgetSwitcher_EventStateSwitcher
   widget=LiveEventViolator
   properties {verticalAlignment=VAlign_Bottom, renderTransform {angle=0, scale=<1 1>, shear=<0 0>, translation=<15 0>}}
   slot {class=WidgetSwitcherSlot, properties {horizontalAlignment=HAlign_Right, verticalAlignment=VAlign_Bottom}}
  }
  LiveViolator {
   widgetClassPath=/Game/UI/Frontend/Showdown/LiveViolator
   bIsVariable=true
   parent=LiveEventViolator
   widget=LiveViolator
   slot {class=BorderSlot, properties {verticalAlignment=VAlign_Bottom}}
  }
  FutureEventViolator {
   widgetClassPath=&CommonBorder
   bIsVariable=true
   parent=WidgetSwitcher_EventStateSwitcher
   widget=FutureEventViolator
   properties {horizontalAlignment=HAlign_Right, verticalAlignment=VAlign_Bottom, padding=<ltrb 0 0 0 0>}
   slot {class=WidgetSwitcherSlot, properties {verticalAlignment=VAlign_Bottom}}
  }
  Text_TimeRemaining {
   widgetClassPath=&CommonTextBlock
   bInherited=true
   bIsVariable=true
   parent=FutureEventViolator
   widget=Text_TimeRemaining
   properties {
    justification=Right
    style=/FortUILibrary/Styles/Typography/_TextStyle_SubH4._TextStyle_SubH4_C
    text=COMING SOON
    font {
     bForceMonospaced=false
     bMaterialIsStencil=false
     fontMaterial=null
     fontObject=&BurbankSmall
     letterSpacing=0
     monospacedWidth=1
     size=24
     skewAmount=0
     typefaceFontName=Black
     outlineSettings {
      bApplyOutlineToDropShadows=false
      bMiteredCorners=false
      bSeparateFillAlpha=false
      outlineMaterial=null
      outlineSize=0
      outlineColor=<rgba 0 0 0 1>
     }
    }
    shadowOffset=<0 0>
    strikeBrush {
     bIsDynamicallyLoaded=false
     drawAs=NoDrawType
     imageType=NoImage
     mirroring=NoMirror
     resourceName=None
     resourceObject=null
     tiling=NoTile
     imageSize=<32 32>
     margin=<ltrb 0 0 0 0>
     outlineSettings {
      bUseBrushTransparency=false
      roundingType=HalfHeightRadius
      width=0
      color {colorUseRule=UseColor_Specified, specifiedColor=<rgba 0 0 0 0>}
      cornerRadii=<0 0 0 0>
     }
     tintColor {colorUseRule=UseColor_Specified, specifiedColor=<rgba 1 1 1 1>}
     uVRegion {bIsValid=false, max=<0 0>, min=<0 0>}
    }
   }
   slot {class=BorderSlot, properties {horizontalAlignment=HAlign_Right, verticalAlignment=VAlign_Bottom, padding=<ltrb 0 0 0 0>}}
  }
  NoViolator {
   widgetClassPath=&CommonBorder
   bIsVariable=true
   parent=WidgetSwitcher_EventStateSwitcher
   widget=NoViolator
   slot {class=WidgetSwitcherSlot, properties {horizontalAlignment=HAlign_Right, verticalAlignment=VAlign_Bottom}}
  }
  InformationViolator {
   widgetClassPath=&CommonBorder
   bIsVariable=true
   parent=WidgetSwitcher_EventStateSwitcher
   widget=InformationViolator
   properties {horizontalAlignment=HAlign_Right, verticalAlignment=VAlign_Bottom, padding=<ltrb 0 0 0 0>}
   slot {class=WidgetSwitcherSlot, properties {verticalAlignment=VAlign_Bottom}}
  }
  SizeBox_0 {
   widgetClassPath=&SizeBox
   parent=InformationViolator
   widget=SizeBox_0
   properties {bOverride_MaxDesiredWidth=true, maxDesiredWidth=750}
   slot {class=BorderSlot, properties {horizontalAlignment=HAlign_Right, verticalAlignment=VAlign_Bottom, padding=<ltrb 0 0 0 0>}}
  }
  Text_InformationViolator {
   widgetClassPath=&CommonTextBlock
   bInherited=true
   bIsVariable=true
   parent=SizeBox_0
   widget=Text_InformationViolator
   properties {
    justification=Right
    scrollStyle=&TextScrollStyle-Base_C
    style=&_TextStyle_SubH4_C
    text=REGISTER NOW
    font {
     bForceMonospaced=false
     bMaterialIsStencil=false
     fontMaterial=null
     fontObject=&BurbankSmall
     letterSpacing=0
     monospacedWidth=1
     size=24
     skewAmount=0
     typefaceFontName=Black
     outlineSettings {
      bApplyOutlineToDropShadows=false
      bMiteredCorners=false
      bSeparateFillAlpha=false
      outlineMaterial=null
      outlineSize=0
      outlineColor=<rgba 0 0 0 1>
     }
    }
    shadowOffset=<0 0>
    strikeBrush {
     bIsDynamicallyLoaded=false
     drawAs=NoDrawType
     imageType=NoImage
     mirroring=NoMirror
     resourceName=None
     resourceObject=null
     tiling=NoTile
     imageSize=<32 32>
     margin=<ltrb 0 0 0 0>
     outlineSettings {
      bUseBrushTransparency=false
      roundingType=HalfHeightRadius
      width=0
      color {colorUseRule=UseColor_Specified, specifiedColor=<rgba 0 0 0 0>}
      cornerRadii=<0 0 0 0>
     }
     tintColor {colorUseRule=UseColor_Specified, specifiedColor=<rgba 1 1 1 1>}
     uVRegion {bIsValid=false, max=<0 0>, min=<0 0>}
    }
   }
   slot {class=SizeBoxSlot}
  }
 ]
}
