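# Sound cue for the front-end store UI, named for a card pick in a showdown.
# A single wave player feeds the Card_Summary wave through a Modulator that randomizes a raised pitch at a fixed, reduced volume.
# The cue uses the Fort_UI sound class with a low volume multiplier.
# A concurrency override allows one instance at a time, stopping the oldest when a new one plays.
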SoundCue /Game/Sounds/Fort_UI_Sounds/FrontEnd/Store/Fort_UI_Showdown_Cardpick_Cue
root=Modulator_0
nodes [
 {id=Modulator_0, children [WavePlayer_13], properties {pitchMax=1.35, pitchMin=1.2, volumeMax=0.3, volumeMin=0.3}}
 {id=WavePlayer_13, properties {soundWaveAssetPtr=/Game/Sounds/Fort_UI_Sounds/FrontEnd/Store/wav/Card_Summary}}
]
properties {
 bOverrideConcurrency=true
 maxDistance=2097152
 soundClassObject=&Fort_UI
 volumeMultiplier=0.1
 concurrencyOverrides {maxCount=1, resolutionRule=StopOldest}
}
