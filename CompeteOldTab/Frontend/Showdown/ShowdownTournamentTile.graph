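# Showdown tournament list tile widget showing an inherited poster and violator, built on FortShowdownTournamentTile.
# On selection it disables hit testing, plays the sizeUp poster animation and calls TournamentListSelect on its parent ShowdownScreen, falling back to BacchusShowdownScreen, then plays a card-pick sound if visible.
# Deselection restores hit testing and runs the size animation back at triple speed.
# EventFade drives a blue tint overlay through changeFocus, using isFaded to skip redundant transitions.
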WidgetBlueprint /CompeteOldTab/Frontend/Showdown/ShowdownTournamentTile
animations [
 changeFocus {
  binding_count=1
  display_label=""
  length_seconds=0.1968
  object=&self:changeFocus
  sequence {bindings [Image_TileFade {
     id=#1
     possessed_class=Image
     tracks [{
       type=MovieSceneFloatTrack
       display_name=Render Opacity
       sections [{type=MovieSceneFloatSection, end_frame=3, start_frame=0, channels [None {type=MovieSceneScriptingFloatChannel, keys [[0,0],[3.9363,1]]}]}]
      }]
    }]}
 }
 sizeUp {
  binding_count=1
  display_label=""
  length_seconds=0.2457
  object=&self:sizeUp
  sequence {
   bindings [PosterSizeBox {
     id=#2
     possessed_class=SizeBox
     tracks [
      {
       type=MovieSceneFloatTrack
       display_name=Width Override
       sections [{type=MovieSceneFloatSection, start_frame=0, channels [None {type=MovieSceneScriptingFloatChannel, keys [[0,378],[14.744,466]]}]}]
      }
      {
       type=MovieSceneFloatTrack
       display_name=Height Override
       sections [{type=MovieSceneFloatSection, channels [None {type=MovieSceneScriptingFloatChannel, keys [[0,543],[14.744,670]]}]}]
      }
     ]
    }]
   display_rate {numerator=60}
  }
 }
]
graphs [EventGraph {nodes [
   n1 = EventFade!() -> (OutputDelegate, then, fadeOut:bool=~) {is=title K2Node_CustomEvent}
   n2 = Get isFaded() -> (isFaded:bool)
   > n3 = Branch(Condition=AND Boolean(A=NOT Boolean(A=@n2), B=@n1.fadeOut))
   > Set isFaded(isFaded:bool=true) -> (Output_Get:bool)
   n4 = Self-Reference()
   n5 = Get changeFocus() -> (changeFocus:Widget Animation Object Reference)
   > Anim Transition!(Exec:Exec, In:bool=@n1.fadeOut, Animation:Widget Animation Object Reference=@n5, PlaySpeed:double=1, Widget:User Widget Object Reference=@n4) -> (ExecOut:Exec) {is=title K2Node_MacroInstance}
   Branch(execute=@n3.else, Condition=AND Boolean(A=Get isFaded(), B=NOT Boolean(A=@n1.fadeOut)))
   > Set isFaded(isFaded=false)
   > Anim Transition(In=@n1.fadeOut, Animation=@n5, PlaySpeed=1, Widget=@n4)
   Event On Selected()
   > SetVisibility(InVisibility=HitTestInvisible)
   n6 = Self-Reference()
   n7 = Get sizeUp() -> (sizeUp:Widget Animation Object Reference)
   > Anim Transition(In=true, Animation=@n7, PlaySpeed=1, Widget=@n6)
   > n8 = FindParentWidgetOfType!(execute, StartingWidget:Widget Object Reference=Self-Reference(), Type:Widget Class Reference=/Script/FortniteUI.FortShowdownScreen) -> (then, ReturnValue:Fort Showdown Screen Object Reference)
   > n9 = Cast To ShowdownScreen!(execute, Object=@n8) -> (then, CastFailed, AsShowdown Screen:Showdown Screen Object Reference) {is=title K2Node_DynamicCast}
   n10 = Self-Reference()
   > n11 = TournamentListSelect!(execute, self:Showdown Screen Object Reference=@n9, CurrentWidget:Widget Object Reference=@n10) -> (then)
   Branch(execute=[@n11.then, @n14.then], Condition=IsVisible())
   > PlaySound2D(Sound=/Game/Sounds/Fort_UI_Sounds/FrontEnd/Store/Fort_UI_Showdown_Cardpick_Cue)
   n12 = FindParentWidgetOfType(execute=@n9.CastFailed, StartingWidget=Self-Reference(), Type=&/BacchusShowdownScreen.BacchusShowdownScreen_C) -> (ReturnValue:Bacchus Showdown Screen Object Reference)
   > n13 = Cast To BacchusShowdownScreen!(execute, Object=@n12) -> (then, CastFailed, AsBacchus Showdown Screen:Bacchus Showdown Screen Object Reference) {is=title K2Node_DynamicCast}
   > n14 = TournamentListSelect(self:Bacchus Showdown Screen Object Reference=@n13, CurrentWidget=@n10)
   Event On Deselected()
   > SetVisibility(InVisibility=Visible)
   > Anim Transition(Animation=@n7, PlaySpeed=3, Widget=@n6)
  ]}]
variables [isFaded {type=boolean}]
widgets {
 info {inheritedWidgetCount=3, parentClass=/Script/FortniteUI.FortShowdownTournamentTile, rootWidgetClass=&Overlay, widgetCount=9}
 widgets [
  Overlay_1 {widgetClassPath=&Overlay, widget=Overlay_1}
  PosterSizeBox {
   widgetClassPath=&SizeBox
   bIsVariable=true
   parent=Overlay_1
   widget=PosterSizeBox
   properties {bOverride_HeightOverride=true, bOverride_WidthOverride=true, heightOverride=543, widthOverride=378}
   slot {class=OverlaySlot, properties {horizontalAlignment=HAlign_Fill, verticalAlignment=VAlign_Bottom, padding=<ltrb 12 0 12 0>}}
  }
  Overlay_4 {
   widgetClassPath=&Overlay
   parent=PosterSizeBox
   widget=Overlay_4
   slot {class=SizeBoxSlot, properties {verticalAlignment=VAlign_Bottom}}
  }
  ScaleBox_0 {widgetClassPath=&ScaleBox, parent=Overlay_4, widget=ScaleBox_0, slot {class=OverlaySlot}}
  Overlay_0 {widgetClassPath=&Overlay, parent=ScaleBox_0, widget=Overlay_0, slot {class=ScaleBoxSlot}}
  ShowdownDetailView_Poster {
   widgetClassPath=&/ShowdownTournamentPosterFront
   bInherited=true
   bIsVariable=true
   parent=Overlay_0
   widget=ShowdownDetailView_Poster
   properties {titleColor {colorUseRule=UseColor_Specified, specifiedColor=<rgba 1 1 1 1>}}
   slot {class=OverlaySlot}
  }
  Image_TileFade {
   widgetClassPath=&Image
   bIsVariable=true
   parent=Overlay_0
   widget=Image_TileFade
   properties {renderOpacity=0, colorAndOpacity=<rgba 0.0307 0.1119 0.5647 0.6>}
   slot {class=OverlaySlot, properties {horizontalAlignment=HAlign_Fill, verticalAlignment=VAlign_Fill}}
  }
  ShowdownDetailView_Violator {
   widgetClassPath=&/ShowdownPosterViolator
   bInherited=true
   bIsVariable=true
   parent=Overlay_4
   widget=ShowdownDetailView_Violator
   properties {renderTransform {angle=0, scale=<1 1>, shear=<0 0>, translation=<0 -46>}}
   slot {class=OverlaySlot, properties {horizontalAlignment=HAlign_Right}}
  }
  InputActionWidget {widgetClassPath=&CommonActionWidget, bInherited=true, widget=null}
 ]
}
