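# Frontend Compete-tab screen (FortShowdownScreen subclass) showing tournaments as a horizontal poster list navigated by arrow buttons or swipes, with an Event Details button.
# On activation it plays an intro animation and sound, then switches between tournament, no-data and placeholder views based on BR registration and list count.
# A context menu built from available regions sets the tournament region.
# Input mode, party leadership and mobile checks adjust visibility and padding.
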WidgetBlueprint /CompeteOldTab/Frontend/Showdown/ShowdownScreen
animations [
 Intro {
  binding_count=4
  display_label=""
  length_seconds=1.8134
  object=&self:Intro
  sequence {bindings [
    SDS_Image_BG {
     id=#1
     possessed_class=Image
     tracks [
      {
       type=MovieSceneWidgetMaterialTrack
       sections [{
         type=MovieSceneParameterSection
         end_frame=36
         start_frame=0
         channels [
          Density {type=MovieSceneScriptingFloatChannel, keys [[0,0.65],[32.4533,1.3399]]}
          Density2 {type=MovieSceneScriptingFloatChannel, keys [[0,0.3259],[23.0477,1.0798]]}
          RightEdge {type=MovieSceneScriptingFloatChannel, keys [[0,1.2544],[15.2587,0.2879]]}
          LeftEdge {type=MovieSceneScriptingFloatChannel, keys [[0,-0.3437],[15.2587,0.0163]]}
         ]
        }]
      }
      {
       type=MovieSceneFloatTrack
       display_name=Render Opacity
       sections [{type=MovieSceneFloatSection, end_frame=36, start_frame=0, channels [None {type=MovieSceneScriptingFloatChannel, keys [[0,0],[36.2683,1]]}]}]
      }
     ]
    }
    Button_Details {
     id=#2
     possessed_class=PrimarySkewButton_C
     tracks [{
       type=MovieScene2DTransformTrack
       sections [{
         type=MovieScene2DTransformSection
         channels [
          Translation.X {type=MovieSceneScriptingFloatChannel}
          Translation.Y {type=MovieSceneScriptingFloatChannel}
          Angle {type=MovieSceneScriptingFloatChannel}
          Scale.X {type=MovieSceneScriptingFloatChannel, keys [[0,0.669],[2.1307,1.1],[8.5403,0.9851],[12.7927,1.0054],[16.6367,1]]}
          Scale.Y {type=MovieSceneScriptingFloatChannel, keys [[0,0.669],[2.1307,1.1],[8.5403,0.9851],[12.7927,1.0054],[16.6367,1]]}
          Shear.X {type=MovieSceneScriptingFloatChannel}
          Shear.Y {type=MovieSceneScriptingFloatChannel}
         ]
        }]
      }]
    }
    CircleArrowButton_Left {
     id=#3
     possessed_class=CircleArrowButton_C
     tracks [{
       type=MovieScene2DTransformTrack
       sections [{
         type=MovieScene2DTransformSection
         channels [
          Translation.X {type=MovieSceneScriptingFloatChannel, keys [[0,-200],[14.7947,0]]}
          Translation.Y {type=MovieSceneScriptingFloatChannel}
          Angle {type=MovieSceneScriptingFloatChannel}
          Scale.X {type=MovieSceneScriptingFloatChannel}
          Scale.Y {type=MovieSceneScriptingFloatChannel}
          Shear.X {type=MovieSceneScriptingFloatChannel}
          Shear.Y {type=MovieSceneScriptingFloatChannel}
         ]
        }]
      }]
    }
    CircleArrowButton_Right {
     id=#4
     possessed_class=CircleArrowButton_C
     tracks [{
       type=MovieScene2DTransformTrack
       sections [{
         type=MovieScene2DTransformSection
         channels [
          Translation.X {type=MovieSceneScriptingFloatChannel, keys [[0,200],[14.7947,0]]}
          Translation.Y {type=MovieSceneScriptingFloatChannel}
          Angle {type=MovieSceneScriptingFloatChannel}
          Scale.X {type=MovieSceneScriptingFloatChannel}
          Scale.Y {type=MovieSceneScriptingFloatChannel}
          Shear.X {type=MovieSceneScriptingFloatChannel}
          Shear.Y {type=MovieSceneScriptingFloatChannel}
         ]
        }]
      }]
    }
   ]}
 }
 Outro {
  binding_count=1
  display_label=""
  length_seconds=0.3
  object=&self:Outro
  sequence {bindings [SDS_Image_BG {
     id=#5
     possessed_class=Image
     tracks [
      {
       type=MovieSceneWidgetMaterialTrack
       sections [{
         type=MovieSceneParameterSection
         end_frame=6
         start_frame=0
         channels [
          Density {type=MovieSceneScriptingFloatChannel, keys [[0,1.3399],[5.9997,0]]}
          Density2 {type=MovieSceneScriptingFloatChannel, keys [[0,1.0798],[5.9997,0]]}
          LeftEdge {type=MovieSceneScriptingFloatChannel, keys [[0,0.0163],[5.9997,0.5063]]}
          RightEdge {type=MovieSceneScriptingFloatChannel, keys [[0,0.2879],[5.9997,0.1879]]}
         ]
        }]
      }
      {
       type=MovieSceneFloatTrack
       display_name=Render Opacity
       sections [{type=MovieSceneFloatSection, end_frame=6, start_frame=0, channels [None {type=MovieSceneScriptingFloatChannel, keys [[0,1],[6,0]]}]}]
      }
     ]
    }]}
 }
]
graphs [
 AddCard {nodes [
   Inputs()
   > n1 = Create Widget!(
    execute:Exec
    Class:User Widget Class Reference=&/ShowdownTournamentTile.ShowdownTournamentTile_C
    OwningPlayer:Player Controller Object Reference=GetOwningPlayer()
    Style:Common Button Style Class Reference=/FortUILibrary/Foundation/Buttons/ButtonStyles/ButtonStyle-TextOnlyBase_Empty.ButtonStyle-TextOnlyBase_Empty_C
    bApplyAlphaOnDisable:bool=True
    bLocked:bool=False
    bSelectable:bool=True
    bShouldSelectUponReceivingFocus:bool=True
    bInteractableWhenSelected:bool=True
    bToggleable:bool=False
    bTriggerClickedAfterSelection:bool=False
    bDisplayInputActionWhenNotInteractable:bool=True
    bHideInputActionWithKeyboard:bool=False
    bShouldUseFallbackDefaultInputAction:bool=True
    bRequiresHold:bool=False
    HoldData:Common UIHold Data Class Reference
    ClickMethod:EButtonClickMethod Enum=MouseDown
    InputPriority:int=0
    TriggeringInputAction:Data Table Row Handle Structure
   ) -> (then:Exec, ReturnValue:Showdown Tournament Tile Object Reference)
   > AddItem!(
    execute
    self:List View Object Reference=Get ListView_TournamentList() -> (ListView_TournamentList:Fort Showdown Tournament List View Object Reference)
    Item:Object Reference=@n1
   ) -> (then)
   > Outputs!(then:Exec)
  ]}
 EventGraph {nodes [
   "On Button Clicked (CircleArrowButton_Left)"!() -> (OutputDelegate, then, Button:Common Button Legacy Object Reference) {is=title K2Node_ComponentBoundEvent}
   > NavigateTournamentLeft(self:Fort Showdown Screen Object Reference)
   "On Button Clicked (CircleArrowButton_Right)"!() -> (OutputDelegate, then, Button:Common Button Legacy Object Reference) {is=title K2Node_ComponentBoundEvent}
   > NavigateTournamentRight(self:Fort Showdown Screen Object Reference)
   "On Swipe Left (SwipePanel)"() {is=title K2Node_ComponentBoundEvent}
   > NavigateTournamentRight()
   "On Swipe Right (SwipePanel)"() {is=title K2Node_ComponentBoundEvent}
   > NavigateTournamentLeft()
   n1 = "On Item Selection Changed (ListView_TournamentList)"!() -> (OutputDelegate, then, Item:Object Reference, bIsSelected:bool) {is=title K2Node_ComponentBoundEvent}
   > Branch(Condition=AND Boolean(A=@n1.bIsSelected, B=Get bIsUsingGamePad() -> (bIsUsingGamePad:bool)))
   > ScrollItemToDesiredOffset!(execute, self:Fort Showdown Tournament List View Object Reference=Get ListView_TournamentList(), InObject:Object Reference=@n1.Item) -> (then)
   "On Context Menu Opened (ContextMenu_RegionSelection)"() {is=title K2Node_ComponentBoundEvent}
   n2 = Get ContextMenuButton!(self:WBP UIKit Context Menu Anchor Button Base Object Reference=Get ContextMenu_RegionSelection() -> (ContextMenu_RegionSelection:WBP UIKit Context Menu Anchor Button Object Reference)) -> (ContextMenuButton:WBP UIKit Button Generic Base Object Reference)
   > Set bIsEnabled!(execute, bIsEnabled:bool=false, self:Widget Object Reference=@n2) -> (then, Output_Get:bool)
   n3 = Get Button_Details() -> (Button_Details:Common Button Base Object Reference)
   > Set bIsEnabled(bIsEnabled=false, self=@n3)
   n4 = Get Button_ChangeFlag() -> (Button_ChangeFlag:Common Button Base Object Reference)
   > Set bIsEnabled(bIsEnabled=false, self=@n4)
   > GenerateContextMenu()
   "On Context Menu Closed (ContextMenu_RegionSelection)"() {is=title K2Node_ComponentBoundEvent}
   > Set bIsEnabled(bIsEnabled=true, self=@n2)
   > Set bIsEnabled(bIsEnabled=true, self=@n3)
   > Set bIsEnabled(bIsEnabled=true, self=@n4)
   n5 = TournamentListSelect!() -> (OutputDelegate, then, CurrentWidget:Widget Object Reference) {is=title K2Node_CustomEvent}
   > n6 = Set SelectedTile(SelectedTile:Widget Object Reference=@n5.CurrentWidget) -> (Output_Get:Widget Object Reference)
   > n7 = Cast To ShowdownTournamentTile!(execute, Object=@n6) -> (then, CastFailed, AsShowdown Tournament Tile:Showdown Tournament Tile Object Reference) {is=title K2Node_DynamicCast}
   n8 = Get ListView_TournamentList()
   > n9 = For Loop(LastIndex=int - int(A=GetNumItems!(self:List View Object Reference=@n8) -> (ReturnValue:int=0), B=1) {is=title K2Node_CallFunction})
   n10 = GetIndexForItem!(
    self:List View Object Reference=@n8
    Item:Object Reference=GetTournamentListEntryData!(self:Fort Showdown Tournament Tile Object Reference=@n7) -> (ReturnValue:Tournament List Entry Data Object Reference)
   ) -> (ReturnValue:int=0)
   n11 = int - int(A=@n10, B=1)
   n12 = int + int(A=@n10, B=1) -> (ReturnValue=0) {is=title K2Node_CommutativeAssociativeBinaryOperator}
   > n13 = Branch(Condition=OR Boolean(
     A="integer < integer"!(A:int=@n9, B:int=@n11) -> (ReturnValue:bool) {is=title K2Node_CallFunction}
     B="integer > integer"(A=@n9, B=@n12) {is=title K2Node_CallFunction}
    ))
   n14 = GetEntryAtIndex!(self:Fort Showdown Tournament List View Object Reference=Get ListView_TournamentList(), SelectedIndex:int=@n9) -> (ReturnValue:User Widget Object Reference)
   > n15 = Cast To ShowdownTournamentTile(Object=@n14)
   > EventFade!(execute, self:Showdown Tournament Tile Object Reference=@n15, fadeOut:bool=true) -> (then)
   n16 = Cast To ShowdownTournamentTile(execute=@n13.else, Object=@n14)
   > EventFade(self=@n16, fadeOut=false)
   Event Construct()
   > n17 = Construct Object from Class(Class=/Script/FortniteUI.UIKitContextMenuActionContext) -> (ReturnValue:UIKit Context Menu Action Context Object Reference)
   > Set ContextMenuActionContext(ContextMenuActionContext:UIKit Context Menu Action Context Object Reference=@n17) -> (Output_Get:UIKit Context Menu Action Context Object Reference)
   > n18 = GetSafeZonePadding!(execute) -> (then, SafePadding:Vector 4 Structure, SafePaddingScale:Vector 2D Structure, SpillOverPadding:Vector 4 Structure)
   n19 = SlotAsOverlaySlot(Widget=Get SDS_Image_BG() -> (SDS_Image_BG:Image Object Reference))
   n20 = Break Margin!(Margin:"Margin Structure (by ref)"=Get Padding!(self:Overlay Slot Object Reference=@n19) -> (Padding:Margin Structure)) -> (Left:float=~, Top:float=~, Right:float=~, Bottom:float=~) {is=title K2Node_BreakStruct}
   n21 = BreakVector4!(InVec:"Vector 4 Structure (by ref)"=@n18) -> (X:double, Y:double, Z:double, W:double)
   > SetPadding!(
    execute
    self:Overlay Slot Object Reference=@n19
    InPadding:Margin Structure=Make Margin(
     Left=float - float(A=@n20, B=@n21) {is=title K2Node_CallFunction}
     Top=float - float(A=@n20.Top, B=@n21.Y)
     Right=float - float(A=@n20.Right, B=@n21.Z)
     Bottom=float - float(A=@n20.Bottom, B=@n21.W)
    )
   ) -> (then)
   > SetScrollbarVisibility!(execute, self:List View Base Object Reference=Get ListView_TournamentList(), InVisibility:ESlateVisibility Enum=Hidden) -> (then)
   > n22 = Branch(Condition=IsMobileGame())
   n23 = SlotAsOverlaySlot(Widget=Get Overlay_RegionButtons() -> (Overlay_RegionButtons:Overlay Object Reference))
   > SetPadding!(execute, self=@n23, InPadding_Left:float=0, InPadding_Top:float=20, InPadding_Right:float=30, InPadding_Bottom:float=24) -> (then)
   n24 = SlotAsOverlaySlot(Widget=Get SDS_SafeZone_Main() -> (SDS_SafeZone_Main:Safe Zone Object Reference))
   > SetPadding(self=@n24, InPadding_Left=0, InPadding_Top=0, InPadding_Right=0, InPadding_Bottom=0)
   n25 = GetViewportSize!() -> (ReturnValue_X:double, ReturnValue_Y:double)
   > Branch(Condition="float > float"(A=float / float(A=@n25, B=@n25.ReturnValue_Y), B=1.34))
   > SetSize!(
    execute
    self:Vertical Box Slot Object Reference=SlotAsVerticalBoxSlot(Widget=Get Spacer_Bottom() -> (Spacer_Bottom:Spacer Object Reference))
    InSize_Value:float=0.1
    InSize_SizeRule:ESlateSizeRule Enum=Fill
   ) -> (then)
   SetPadding(execute=@n22.else, self=@n23, InPadding_Left=24, InPadding_Top=24, InPadding_Right=24, InPadding_Bottom=0)
   > SetPadding(self=@n24, InPadding_Left=0, InPadding_Top=0, InPadding_Right=0, InPadding_Bottom=112)
   Event On Deactivated()
   > SetTriggeringInputAction!(
    execute
    self:Common Button Base Object Reference=Get Button_Details()
    InputActionRow:"Data Table Row Handle Structure (by ref)"=Make Data Table Row Handle!(DataTable:Data Table Object Reference, RowName:Name=None) -> (DataTableRowHandle:Data Table Row Handle Structure) {is=title K2Node_MakeStruct}
   ) -> (then)
   Event PreConstruct()
   > SetColorAndOpacity(
    self=Get CommonTextBlock_RegionText() -> (CommonTextBlock_RegionText:Common Text Object Reference)
    InColorAndOpacity=Get PrimaryColor() -> (PrimaryColor:Slate Color Structure)
   )
   Event On Activated()
   > SetTriggeringInputAction(
    self=Get Button_Details()
    InputActionRow=Make Data Table Row Handle(DataTable=&InputActionDataTable, RowName=EventDetails)
   )
   > SetFrontEndCamera!(
    execute
    self:Fort Front End Context Object Reference=Get FortFrontEndContext!() -> (ReturnValue:Fort Front End Context Object Reference) {is=title K2Node_GetBlueprintContextEx}
    NewState:EFrontEndCamera Enum=FrontendDefault
    TransitionParams:View Target Transition Params Structure=(BlendTime=0,BlendFunction=VTBlend_Cubic,BlendExp=2,bLockOutgoing=False)
   ) -> (then)
   > n26 = Branch(Condition=IsBRRegistered!(self:Fort Showdown Screen Object Reference) -> (ReturnValue:bool))
   n27 = GetRuntimeOptions()
   > SetVisibility(
    self=Get Overlay_ChangeFlag() -> (Overlay_ChangeFlag:Overlay Object Reference)
    InVisibility=Select(
     Option 0=Collapsed
     Option 1=Visible
     Index=Get bEnableFlagSelection!(self:Fort Runtime Options Object Reference=@n27) -> (bEnableFlagSelection:bool)
    )
   )
   > PlayAnimation(InAnimation=Get Intro())
   > SpawnSound2D!(
    execute
    Sound:Sound Base Object Reference=Get IntroWhoosh() -> (IntroWhoosh:Sound Base Object Reference)
    VolumeMultiplier:float=1
    PitchMultiplier:float=1
    StartTime:float=0
    ConcurrencySettings:Sound Concurrency Object Reference
    bPersistAcrossLevelTransition:bool=false
    bAutoDestroy:bool=true
   ) -> (then, ReturnValue:Audio Component Object Reference)
   n28 = GetNumItems(self=Get ListView_TournamentList()) -> (ReturnValue=0)
   > SetActiveWidget(
    self:Widget Switcher Object Reference=Get CommonWidgetSwitcher_ErrorSwitcher() -> (CommonWidgetSwitcher_ErrorSwitcher:Common Widget Switcher Legacy Object Reference)
    Widget=Select(
     Option 0:Widget Object Reference=Get Overlay_NoTournamentContent() -> (Overlay_NoTournamentContent:Overlay Object Reference)
     Option 1:Widget Object Reference=Get TournamentsContent() -> (TournamentsContent:Overlay Object Reference)
     Index="integer > integer"(A=@n28, B=0)
    ) -> (ReturnValue:Widget Object Reference)
   )
   > SetScrollbarVisibility(self=Get ListView_TournamentList(), InVisibility=Hidden)
   > SetVisibility(
    self=Get ContextMenu_RegionSelection()
    InVisibility=Select(
     Option 0:ESlateVisibility Enum=Collapsed
     Option 1:ESlateVisibility Enum=SelfHitTestInvisible
     Index=ShouldDisplayRegionButton!(self:Fort Showdown Screen Object Reference) -> (ReturnValue:bool)
    ) -> (ReturnValue:ESlateVisibility Enum)
   )
   > UpdateText!(
    execute
    self:WBP UIKit Context Menu Anchor Button Base Object Reference=Get ContextMenu_RegionSelection()
    Text:Text="To Text (String)"(InString=GetCurrentRegionID!(self:Fort Showdown Screen Object Reference) -> (ReturnValue:String))
   ) -> (then)
   > GenerateContextMenu()
   SetActiveWidget(
    execute=@n26.else
    self=Get CommonWidgetSwitcher_ErrorSwitcher()
    Widget=Get PlaceholderScreen() -> (PlaceholderScreen:Overlay Object Reference)
   )
   > SetVisibility(self=Get Overlay_ChangeFlag(), InVisibility=Hidden)
   > SetVisibility(self=Get Overlay_RegionButtons(), InVisibility=Hidden)
   > SetScrollbarVisibility(self=Get ListView_TournamentList(), InVisibility=Hidden)
   n29 = Event OnInputModeChanged!() -> (OutputDelegate, then, bUsingGamepad:bool) {is=title K2Node_Event}
   > Set bIsUsingGamePad(bIsUsingGamePad:bool=@n29.bUsingGamepad) -> (Output_Get:bool)
   > SetVisibility(
    self=[
     Get CircleArrowButton_Left() -> (CircleArrowButton_Left:Circle Arrow Button Object Reference)
     Get CircleArrowButton_Right() -> (CircleArrowButton_Right:Circle Arrow Button Object Reference)
    ]
    InVisibility=Select(Option 0=SelfHitTestInvisible, Option 1=Collapsed, Index=@n29.bUsingGamepad)
   )
   n30 = Get ContextMenu_RegionSelection()
   > CloseMenu(self:WBP UIKit Context Menu Anchor Button Base Object Reference=@n30)
   n31 = IsMobileGame()
   n32 = Get bIsUsingGamePad()
   > SetPadding(
    self:User Widget Object Reference=@n30
    InPadding_Left=SelectFloat(A=24, B=0, bPickA=AND Boolean(A=@n31, B=@n32))
    InPadding_Top=0
    InPadding_Right=0
    InPadding_Bottom=0
   )
   Event OnInitialized()
   > Branch(Condition=IsMobileGame())
   > Set Min Height!(execute, Min Height:int=0, self:WBP UIKit Context Menu Anchor Button Base Object Reference=Get ContextMenu_RegionSelection()) -> (then, Output_Get:int=0)
   n33 = Event SelectedFrontTournament!() -> (OutputDelegate, then, bFrontIsSelected:bool) {is=title K2Node_Event}
   > n34 = Branch(Condition=@n33.bFrontIsSelected)
   n35 = Get CircleArrowButton_Left()
   > SetVisibility(self=@n35, InVisibility=Hidden)
   SetVisibility(execute=@n34.else, self=@n35, InVisibility=Visible)
   n36 = Event SelectedEndTournament!() -> (OutputDelegate, then, bEndIsSelected:bool) {is=title K2Node_Event}
   > n37 = Branch(Condition=@n36.bEndIsSelected)
   n38 = Get CircleArrowButton_Right()
   > SetVisibility(self=@n38, InVisibility=Hidden)
   SetVisibility(execute=@n37.else, self=@n38, InVisibility=Visible)
   n39 = Event On Region Changed!() -> (OutputDelegate, then, NewRegionID:String) {is=title K2Node_Event}
   > UpdateText(self=Get ContextMenu_RegionSelection(), Text="To Text (String)"(InString=@n39.NewRegionID))
   n40 = Event On Party Leader Changed!() -> (OutputDelegate, then, bIsPartyLeader:bool) {is=title K2Node_Event}
   n41 = ShouldDisplayRegionButton()
   > SetVisibility(
    self=Get ContextMenu_RegionSelection()
    InVisibility=Select(Option 0=Collapsed, Option 1=SelfHitTestInvisible, Index=AND Boolean(A=@n40.bIsPartyLeader, B=@n41))
   )
  ]}
 GenerateContextMenu {nodes [
   GenerateContextMenu() {is=title K2Node_FunctionEntry}
   > RemoveAllInteractions(self:UIKit Context Menu Action Context Object Reference=Get ContextMenuActionContext() -> (ContextMenuActionContext:UIKit Context Menu Action Context Object Reference))
   > n1 = For Each Loop!(
    Exec
    Array:Array of Showdown Screen Region Wrapper Structures=GetAvailableRegions!(self:Fort Showdown Screen Object Reference) -> (ReturnValue:Array of Showdown Screen Region Wrapper Structures)
   ) -> (LoopBody, Array Element_DisplayName:String, Array Element_RegionID:String, Array Index, Completed)
   > AddInteraction!(
    execute
    self:UIKit Context Menu Action Context Object Reference=Get ContextMenuActionContext()
    InActionInfo_DisplayName:Text="To Text (String)"(InString=@n1.Array Element_RegionID)
    InActionInfo_IconTexture:Texture 2D Soft Object Reference
    InActionInfo_IconMaterial:Material Instance Soft Object Reference
    InActionInfo_SortPriority:int=1000
    InActionInfo_bIsAvailable:bool=True
    InActionInfo_bIsEnabled:bool=True
    InActionInfo_bIsSelected:bool=False
    InActionInfo_bAddDivider:bool=False
    InActionInfo_bPositionDividerAfter:bool=False
    InActionInfo_bCloseMenuOnSelect:bool=False
    InActionInfo_InteractionNameID:Name=String To Name(InString=@n1.Array Element_RegionID)
    InActionInfo_Interaction:Data Table Row Handle Structure=(DataTable=None,RowName="")
    InActionInfo_PressedDelegate:Delegate="(null).None"
    InActionInfo_ReleasedDelegate:Delegate="(null).None"
    InInteractionDelegate:"Delegate (by ref)"=Create Event()
   ) -> (then)
   SetContext!(
    execute=@n1.Completed
    self:WBP UIKit Context Menu Anchor Button Base Object Reference=Get ContextMenu_RegionSelection()
    InteractionContext_Data:UIKit Context Menu Action Context Object Reference=Get ContextMenuActionContext()
   ) -> (then)
  ]}
 OnContextMenuClicked {nodes [
   n1 = OnContextMenuClicked!() -> (then, NameID:Name) {is=title K2Node_FunctionEntry}
   > SetRegionID!(execute, self:Fort Showdown Screen Object Reference, InRegionID:String="To String (Name)"(InName=@n1)) -> (then)
  ]}
]
variables [
 ContextMenuActionContext {type=UIKitContextMenuActionContext, default=null}
 IntroWhoosh {type=SoundBase, default=/Game/Sounds/Fort_UI_Sounds/FrontEnd/Tabs/UI_Whoosh_TopBar_Tab_Compete_Cue}
 PrimaryColor {type=SlateColor, default {colorUseRule=UseColor_Specified, specifiedColor=<rgba 0 0.9734 1 1>}}
 SelectedTile {type=Widget, default=null}
]
widgets {
 info {inheritedWidgetCount=6, parentClass=/Script/FortniteUI.FortShowdownScreen, rootWidgetClass=&Overlay, widgetCount=36}
 widgets [
  SDS_Overlay {widgetClassPath=&Overlay, widget=SDS_Overlay}
  SDS_Image_BG {
   widgetClassPath=&Image
   bIsVariable=true
   parent=SDS_Overlay
   widget=SDS_Image_BG
   properties {
    visibility=HitTestInvisible
    brush {
     bIsDynamicallyLoaded=false
     drawAs=Image
     imageType=NoImage
     mirroring=NoMirror
     resourceName=None
     resourceObject=/CompeteOldTab/Frontend/Showdown/Materials/M_UI_Showdown_LineupBG_Inst
     tiling=NoTile
     imageSize=<32 32>
     margin=<ltrb 0 0 0 0>
     outlineSettings {
      bUseBrushTransparency=false
      roundingType=HalfHeightRadius
      width=0
      color {colorUseRule=UseColor_Specified, specifiedColor=<rgba 0 0 0 0>}
      cornerRadii=<0 0 0 0>
     }
     tintColor {colorUseRule=UseColor_Specified, specifiedColor=<rgba 1 1 1 1>}
     uVRegion {bIsValid=false, max=<0 0>, min=<0 0>}
    }
   }
   slot {class=OverlaySlot, properties {horizontalAlignment=HAlign_Fill, verticalAlignment=VAlign_Fill, padding=<ltrb -40 -110 -40 -60>}}
  }
  SDS_SafeZone_Main {
   widgetClassPath=&SafeZone
   bIsVariable=true
   parent=SDS_Overlay
   widget=SDS_SafeZone_Main
   slot {class=OverlaySlot, properties {horizontalAlignment=HAlign_Fill, verticalAlignment=VAlign_Fill, padding=<ltrb 0 0 0 112>}}
  }
  ScaleBox_2 {
   widgetClassPath=&ScaleBox
   parent=SDS_SafeZone_Main
   widget=ScaleBox_2
   properties {stretch=ScaleBySafeZone}
   slot {class=SafeZoneSlot}
  }
  OverlayMain {
   widgetClassPath=&Overlay
   parent=ScaleBox_2
   widget=OverlayMain
   properties {navigation=&self:WidgetTree.OverlayMain.WidgetNavigation_0}
   slot {class=ScaleBoxSlot, properties {horizontalAlignment=HAlign_Fill, verticalAlignment=VAlign_Fill}}
  }
  CommonBorder_458 {
   widgetClassPath=&CommonBorder
   parent=OverlayMain
   widget=CommonBorder_458
   properties {padding=<ltrb 0 0 0 0>}
   slot {class=OverlaySlot, properties {horizontalAlignment=HAlign_Fill, verticalAlignment=VAlign_Fill}}
  }
  CommonWidgetSwitcher_ErrorSwitcher {
   widgetClassPath=&CommonWidgetSwitcherLegacy
   bIsVariable=true
   parent=CommonBorder_458
   widget=CommonWidgetSwitcher_ErrorSwitcher
   slot {class=BorderSlot, properties {padding=<ltrb 0 0 0 0>}}
  }
  PlaceholderScreen {
   widgetClassPath=&Overlay
   bIsVariable=true
   parent=CommonWidgetSwitcher_ErrorSwitcher
   widget=PlaceholderScreen
   slot {class=WidgetSwitcherSlot}
  }
  WBP_CompetePlaceholderTab {
   widgetClassPath=/CompeteOldTab/Frontend/Showdown/CompetePlaceholders/WBP_CompetePlaceholderTab
   bIsVariable=true
   parent=PlaceholderScreen
   widget=WBP_CompetePlaceholderTab
   slot {class=OverlaySlot, properties {horizontalAlignment=HAlign_Center, verticalAlignment=VAlign_Center}}
  }
  TournamentsContent {
   widgetClassPath=&Overlay
   bIsVariable=true
   parent=CommonWidgetSwitcher_ErrorSwitcher
   widget=TournamentsContent
   slot {class=WidgetSwitcherSlot}
  }
  SwipePanel {
   widgetClassPath=/Script/FortniteUI.FortSwipePanel
   bIsVariable=true
   parent=TournamentsContent
   widget=SwipePanel
   properties {bBeginSwipeOnPointerEnter=false, bConsumePointerInput=false, swipeThreshold=<12 12>}
   slot {class=OverlaySlot, properties {horizontalAlignment=HAlign_Fill, verticalAlignment=VAlign_Fill, padding=<ltrb 0 75 0 75>}}
  }
  VerticalBox_0 {
   widgetClassPath=&VerticalBox
   parent=TournamentsContent
   widget=VerticalBox_0
   slot {class=OverlaySlot, properties {horizontalAlignment=HAlign_Fill, verticalAlignment=VAlign_Fill}}
  }
  Spacer_Top {
   widgetClassPath=&Spacer
   parent=VerticalBox_0
   widget=Spacer_Top
   slot {class=VerticalBoxSlot, properties {size {sizeRule=Fill, value=0.6}}}
  }
  SizeBox_Posters {
   widgetClassPath=&SizeBox
   bIsVariable=true
   parent=VerticalBox_0
   widget=SizeBox_Posters
   properties {bOverride_HeightOverride=true, bOverride_WidthOverride=true, heightOverride=670, widthOverride=466}
   slot {class=VerticalBoxSlot, properties {verticalAlignment=VAlign_Bottom}}
  }
  ListView_TournamentList {
   widgetClassPath=/Script/FortniteUI.FortShowdownTournamentListView
   bInherited=true
   bIsVariable=true
   parent=SizeBox_Posters
   widget=ListView_TournamentList
   properties {
    bCanUseDesiredScrollOffset=true
    bEnableRightClickScrolling=false
    bReturnFocusToSelection=true
    clipping=Inherit
    consumeMouseWheel=Never
    entryWidgetClass=&ShowdownTournamentTile_C
    entryWidthSize=350
    navigation=&self:WidgetTree.ListView_TournamentList.WidgetNavigation_0
    numPaddingRowLeft=2
    numPaddingRowRight=2
    orientation=Orient_Horizontal
   }
   slot {class=SizeBoxSlot}
  }
  Spacer_Bottom_1 {
   widgetClassPath=&Spacer
   parent=VerticalBox_0
   widget=Spacer_Bottom_1
   properties {size=<1 24>}
   slot {class=VerticalBoxSlot, properties {size {sizeRule=Automatic, value=0.4}}}
  }
  Button_Details {
   widgetClassPath=/FortUILibrary/UIKit/Buttons/WBP_UIKit_Button_Loud
   bInherited=true
   bIsVariable=true
   parent=VerticalBox_0
   widget=Button_Details
   properties {bIsFocusable=false, buttonSize=Large, text=Event Details}
   slot {class=VerticalBoxSlot, properties {horizontalAlignment=HAlign_Center, verticalAlignment=VAlign_Center, padding=<ltrb 0 15 0 0>}}
  }
  Spacer_Bottom {
   widgetClassPath=&Spacer
   bIsVariable=true
   parent=VerticalBox_0
   widget=Spacer_Bottom
   slot {class=VerticalBoxSlot, properties {size {sizeRule=Fill, value=0.4}}}
  }
  CircleArrowButton_Right {
   widgetClassPath=&/CircleArrowButton
   bIsVariable=true
   parent=TournamentsContent
   widget=CircleArrowButton_Right
   properties {
    onButtonBaseClicked=(&self:WidgetTree.CircleArrowButton_Right.HandleOnButtonClicked)
    onButtonBaseDoubleClicked=(&self:WidgetTree.CircleArrowButton_Right.HandleOnButtonDoubleClicked)
    onButtonBaseHovered=(&self:WidgetTree.CircleArrowButton_Right.HandleOnButtonHovered)
    onButtonBaseUnhovered=(&self:WidgetTree.CircleArrowButton_Right.HandleOnButtonUnhovered)
    onSelectedChangedBase=(&self:WidgetTree.CircleArrowButton_Right.HandleOnSelectedChanged)
    hoveredSlateSoundOverride {resourceObject=/Game/Sounds/Fort_UI_Sounds/FrontEnd/New_FrontEnd/Fortnite_UI_Map_Hover_04_SC_Cue}
    padding=<ltrb 30 0 30 0>
    pressedSlateSoundOverride {resourceObject=/Game/Sounds/Fort_UI_Sounds/FrontEnd/Fort_Athena_LockerItem_Pressed_02}
   }
   slot {class=OverlaySlot, properties {horizontalAlignment=HAlign_Right, verticalAlignment=VAlign_Center}}
  }
  CircleArrowButton_Left {
   widgetClassPath=&CircleArrowButton
   bIsVariable=true
   parent=TournamentsContent
   widget=CircleArrowButton_Left
   properties {
    flip=true
    onButtonBaseClicked=(&self:WidgetTree.CircleArrowButton_Left.HandleOnButtonClicked)
    onButtonBaseDoubleClicked=(&self:WidgetTree.CircleArrowButton_Left.HandleOnButtonDoubleClicked)
    onButtonBaseHovered=(&self:WidgetTree.CircleArrowButton_Left.HandleOnButtonHovered)
    onButtonBaseUnhovered=(&self:WidgetTree.CircleArrowButton_Left.HandleOnButtonUnhovered)
    onSelectedChangedBase=(&self:WidgetTree.CircleArrowButton_Left.HandleOnSelectedChanged)
    hoveredSlateSoundOverride {resourceObject=/Game/Sounds/Fort_UI_Sounds/FrontEnd/Fort_Athena_DirectAcquisition_Hover_Cue_02}
    padding=<ltrb 30 0 30 0>
    pressedSlateSoundOverride {resourceObject=&Fort_Athena_LockerItem_Pressed_02}
   }
   slot {class=OverlaySlot, properties {verticalAlignment=VAlign_Center}}
  }
  Overlay_NoTournamentContent {
   widgetClassPath=&Overlay
   bIsVariable=true
   parent=CommonWidgetSwitcher_ErrorSwitcher
   widget=Overlay_NoTournamentContent
   slot {class=WidgetSwitcherSlot}
  }
  VerticalBox_2 {
   widgetClassPath=&VerticalBox
   parent=Overlay_NoTournamentContent
   widget=VerticalBox_2
   slot {class=OverlaySlot, properties {horizontalAlignment=HAlign_Center, verticalAlignment=VAlign_Center}}
  }
  Image_0 {
   widgetClassPath=&Image
   parent=VerticalBox_2
   widget=Image_0
   properties {brush {
     bIsDynamicallyLoaded=false
     drawAs=Image
     imageType=NoImage
     mirroring=NoMirror
     resourceName=None
     resourceObject=/BRCosmetics/UI/Foundation/Textures/Icons/Emoji/Season_3/T-Item-Emoji-Kaboom
     tiling=NoTile
     imageSize=<128 128>
     margin=<ltrb 0 0 0 0>
     outlineSettings {
      bUseBrushTransparency=false
      roundingType=HalfHeightRadius
      width=0
      color {colorUseRule=UseColor_Specified, specifiedColor=<rgba 0 0 0 0>}
      cornerRadii=<0 0 0 0>
     }
     tintColor {colorUseRule=UseColor_Specified, specifiedColor=<rgba 1 1 1 1>}
     uVRegion {bIsValid=false, max=<0 0>, min=<0 0>}
    }}
   slot {class=VerticalBoxSlot, properties {horizontalAlignment=HAlign_Center, verticalAlignment=VAlign_Top, padding=<ltrb 0 -128 0 0>}}
  }
  NoTournamentsContent {
   widgetClassPath=&CommonTextBlock
   bIsVariable=true
   parent=VerticalBox_2
   widget=NoTournamentsContent
   properties {
    justification=Center
    text="We're having trouble retrieving data from the Events Neural Network. \r\nWe'll try unplugging it and plugging it back in. \r\n\r\nPlease try back later."
    wrapTextAt=1400
    font {
     bForceMonospaced=false
     bMaterialIsStencil=false
     fontMaterial=null
     fontObject=&BurbankSmall
     letterSpacing=0
     monospacedWidth=1
     size=24
     skewAmount=0
     typefaceFontName=NotoSans-Regular
     outlineSettings {
      bApplyOutlineToDropShadows=false
      bMiteredCorners=false
      bSeparateFillAlpha=false
      outlineMaterial=null
      outlineSize=0
      outlineColor=<rgba 0 0 0 1>
     }
    }
    shadowColorAndOpacity=<rgba 0.0053 0.0248 0.1146 0.2>
    shadowOffset=<8 8>
   }
   slot {class=VerticalBoxSlot}
  }
  Overlay_RegionButtons {
   widgetClassPath=&Overlay
   bIsVariable=true
   parent=OverlayMain
   widget=Overlay_RegionButtons
   slot {class=OverlaySlot, properties {horizontalAlignment=HAlign_Right, verticalAlignment=VAlign_Bottom, padding=<ltrb 0 20 30 -48>}}
  }
  HorizontalBox_130 {widgetClassPath=&HorizontalBox, parent=Overlay_RegionButtons, widget=HorizontalBox_130, slot {class=OverlaySlot}}
  Overlay_ChangeFlag {
   widgetClassPath=&Overlay
   bIsVariable=true
   parent=HorizontalBox_130
   widget=Overlay_ChangeFlag
   slot {class=HorizontalBoxSlot, properties {verticalAlignment=VAlign_Bottom, padding=<ltrb 20 0 20 0>}}
  }
  Button_ChangeFlag {
   widgetClassPath=&WBP_UIKit_Button_Regular
   bInherited=true
   bIsVariable=true
   parent=Overlay_ChangeFlag
   widget=Button_ChangeFlag
   properties {
    text=Change Flag
    hoveredSlateSoundOverride {resourceObject=/Game/Sounds/Fort_UI_Sounds/FrontEnd/Leaderboard/Fort_Athena_LeaderboardTab_Hover_Cue}
    lockedHoveredSlateSoundOverride {resourceObject=&Fort_Athena_LeaderboardTab_Hover_Cue}
    lockedPressedSlateSoundOverride {resourceObject=/Game/Sounds/Fort_UI_Sounds/FrontEnd/Leaderboard/Fortnite_UI_Leaderboard_Tile_Click_Cue}
    pressedSlateSoundOverride {resourceObject=&Fortnite_UI_Leaderboard_Tile_Click_Cue}
    selectedHoveredSlateSoundOverride {resourceObject=&Fort_Athena_LeaderboardTab_Hover_Cue}
    selectedPressedSlateSoundOverride {resourceObject=&Fortnite_UI_Leaderboard_Tile_Click_Cue}
    triggeringInputAction {dataTable=&InputActionDataTable, rowName=GenericFaceButtonLeft}
   }
   slot {class=OverlaySlot, properties {horizontalAlignment=HAlign_Fill}}
  }
  SizeBox_0 {
   widgetClassPath=&SizeBox
   parent=Overlay_ChangeFlag
   widget=SizeBox_0
   properties {bOverride_HeightOverride=true, bOverride_WidthOverride=true, heightOverride=32, widthOverride=32}
   slot {class=OverlaySlot, properties {horizontalAlignment=HAlign_Right, verticalAlignment=VAlign_Center, padding=<ltrb 0 0 -16 0>}}
  }
  Image_YourFlag {
   widgetClassPath=/Script/FortniteUI.FortFlagImage
   bInherited=true
   bIsVariable=true
   parent=SizeBox_0
   widget=Image_YourFlag
   properties {
    materialTextureParamName=Texture
    brush {
     bIsDynamicallyLoaded=false
     drawAs=Image
     imageType=NoImage
     mirroring=NoMirror
     resourceName=None
     resourceObject=/Game/UI/Events/Leaderboards/Materials/M_UI_FlagMaterial_Inst
     tiling=NoTile
     imageSize=<128 128>
     margin=<ltrb 0 0 0 0>
     outlineSettings {
      bUseBrushTransparency=false
      roundingType=HalfHeightRadius
      width=0
      color {colorUseRule=UseColor_Specified, specifiedColor=<rgba 0 0 0 0>}
      cornerRadii=<0 0 0 0>
     }
     tintColor {colorUseRule=UseColor_Specified, specifiedColor=<rgba 1 1 1 1>}
     uVRegion {bIsValid=false, max=<0 0>, min=<0 0>}
    }
   }
   slot {class=SizeBoxSlot}
  }
  VerticalBox_1 {
   widgetClassPath=&VerticalBox
   parent=HorizontalBox_130
   widget=VerticalBox_1
   slot {class=HorizontalBoxSlot, properties {horizontalAlignment=HAlign_Left, verticalAlignment=VAlign_Top}}
  }
  CommonTextBlock_RegionTitle {
   widgetClassPath=&CommonTextBlock
   parent=VerticalBox_1
   widget=CommonTextBlock_RegionTitle
   properties {
    style=/FortUILibrary/Styles/Typography/_TextStyle_NormalTag._TextStyle_NormalTag_C
    text=TOURNAMENT REGION
    font {
     bForceMonospaced=false
     bMaterialIsStencil=false
     fontMaterial=null
     fontObject=&BurbankSmall
     letterSpacing=20
     monospacedWidth=1
     size=15
     skewAmount=0
     typefaceFontName=Black
     outlineSettings {
      bApplyOutlineToDropShadows=false
      bMiteredCorners=false
      bSeparateFillAlpha=false
      outlineMaterial=null
      outlineSize=0
      outlineColor=<rgba 0 0 0 1>
     }
    }
    shadowOffset=<0 0>
    strikeBrush {
     bIsDynamicallyLoaded=false
     drawAs=NoDrawType
     imageType=NoImage
     mirroring=NoMirror
     resourceName=None
     resourceObject=null
     tiling=NoTile
     imageSize=<32 32>
     margin=<ltrb 0 0 0 0>
     outlineSettings {
      bUseBrushTransparency=false
      roundingType=HalfHeightRadius
      width=0
      color {colorUseRule=UseColor_Specified, specifiedColor=<rgba 0 0 0 0>}
      cornerRadii=<0 0 0 0>
     }
     tintColor {colorUseRule=UseColor_Specified, specifiedColor=<rgba 1 1 1 1>}
     uVRegion {bIsValid=false, max=<0 0>, min=<0 0>}
    }
   }
   slot {class=VerticalBoxSlot, properties {horizontalAlignment=HAlign_Right}}
  }
  Border_CTACyan {
   widgetClassPath=&Border
   parent=VerticalBox_1
   widget=Border_CTACyan
   properties {
    background {
     bIsDynamicallyLoaded=false
     drawAs=Image
     imageType=NoImage
     mirroring=NoMirror
     resourceName=None
     resourceObject=null
     tiling=NoTile
     imageSize=<32 32>
     margin=<ltrb 0 0 0 0>
     outlineSettings {
      bUseBrushTransparency=false
      roundingType=HalfHeightRadius
      width=0
      color {colorUseRule=UseColor_Specified, specifiedColor=<rgba 0 0 0 0>}
      cornerRadii=<0 0 0 0>
     }
     tintColor {colorUseRule=UseColor_Specified, specifiedColor=<rgba 1 1 1 0>}
     uVRegion {bIsValid=false, max=<0 0>, min=<0 0>}
    }
    contentColorAndOpacity=<rgba 0.0331 0.7011 1 1>
    padding=<ltrb 0 0 0 0>
   }
   slot {class=VerticalBoxSlot, properties {horizontalAlignment=HAlign_Right}}
  }
  CommonTextBlock_RegionText {
   widgetClassPath=&CommonTextBlock
   bInherited=true
   bIsVariable=true
   parent=Border_CTACyan
   widget=CommonTextBlock_RegionText
   properties {
    style=/FortUILibrary/Styles/Typography/_TextStyle_SubH4._TextStyle_SubH4_C
    text=NORTH AMERICA
    textTransformPolicy=ToUpper
    font {
     bForceMonospaced=false
     bMaterialIsStencil=false
     fontMaterial=null
     fontObject=&BurbankSmall
     letterSpacing=0
     monospacedWidth=1
     size=24
     skewAmount=0
     typefaceFontName=Black
     outlineSettings {
      bApplyOutlineToDropShadows=false
      bMiteredCorners=false
      bSeparateFillAlpha=false
      outlineMaterial=null
      outlineSize=0
      outlineColor=<rgba 0 0 0 1>
     }
    }
    shadowOffset=<0 0>
    strikeBrush {
     bIsDynamicallyLoaded=false
     drawAs=NoDrawType
     imageType=NoImage
     mirroring=NoMirror
     resourceName=None
     resourceObject=null
     tiling=NoTile
     imageSize=<32 32>
     margin=<ltrb 0 0 0 0>
     outlineSettings {
      bUseBrushTransparency=false
      roundingType=HalfHeightRadius
      width=0
      color {colorUseRule=UseColor_Specified, specifiedColor=<rgba 0 0 0 0>}
      cornerRadii=<0 0 0 0>
     }
     tintColor {colorUseRule=UseColor_Specified, specifiedColor=<rgba 1 1 1 1>}
     uVRegion {bIsValid=false, max=<0 0>, min=<0 0>}
    }
   }
   slot {class=BorderSlot, properties {padding=<ltrb 0 0 0 0>}}
  }
  ContextMenu_RegionSelection {
   widgetClassPath=/FortUILibrary/UIKit/ContextMenu/WBP_UIKit_ContextMenuAnchorButton
   bIsVariable=true
   parent=VerticalBox_1
   widget=ContextMenu_RegionSelection
   properties {
    button Background Material=/FortUILibrary/UIKit/Buttons/MaterialInstances/CTAButtons/MI_UIKit_BgLoud
    buttonClass=&/WBP_Showdown_RegionButton.WBP_Showdown_RegionButton_C
    closeMenuOnContextClicked=true
    menuPlacement=MenuPlacement_AboveAnchor
    min Height=56
    min Width=310
    button Text Font {
     bForceMonospaced=false
     bMaterialIsStencil=false
     fontMaterial=/FortUILibrary/UIKit/Buttons/MaterialInstances/CTAButtons/MI_UIKit_TextLoud
     fontObject=&Burbank
     letterSpacing=48
     monospacedWidth=1
     size=21
     skewAmount=0
     typefaceFontName=BurbankRegular-Black
     outlineSettings {
      bApplyOutlineToDropShadows=false
      bMiteredCorners=false
      bSeparateFillAlpha=false
      outlineMaterial=null
      outlineSize=0
      outlineColor=<rgba 0 0 0 1>
     }
    }
    triggering Input Action {dataTable=&InputActionDataTable, rowName=GenericFaceButtonTop}
   }
   slot {class=VerticalBoxSlot, properties {horizontalAlignment=HAlign_Right, padding=<ltrb 0 14 0 0>}}
  }
  Image_LoadingSpinner {
   widgetClassPath=&Image
   bInherited=true
   bIsVariable=true
   parent=SDS_Overlay
   widget=Image_LoadingSpinner
   properties {
    visibility=Collapsed
    brush {
     bIsDynamicallyLoaded=false
     drawAs=Image
     imageType=NoImage
     mirroring=NoMirror
     resourceName=None
     resourceObject=/FortUILibrary/UIKit/Throbber/MI_UIKit_Monochrome_Throbber
     tiling=NoTile
     imageSize=<64 64>
     margin=<ltrb 0 0 0 0>
     outlineSettings {
      bUseBrushTransparency=false
      roundingType=HalfHeightRadius
      width=0
      color {colorUseRule=UseColor_Specified, specifiedColor=<rgba 0 0 0 0>}
      cornerRadii=<0 0 0 0>
     }
     tintColor {colorUseRule=UseColor_Specified, specifiedColor=<rgba 1 1 1 1>}
     uVRegion {bIsValid=false, max=<0 0>, min=<0 0>}
    }
   }
   slot {class=OverlaySlot, properties {horizontalAlignment=HAlign_Center, verticalAlignment=VAlign_Center}}
  }
 ]
}
