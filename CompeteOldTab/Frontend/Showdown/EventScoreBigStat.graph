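# Widget for the tournament (Showdown) event score UI that shows one large stat: a skewed label, a big number and an optional goal score.
# SetData fills the label and value, then formats the context value through the supplied formatter text into a rich text block.
# That goal text is visible for a positive context, collapsed at zero and hidden when negative.
# For non-negative contexts, its border's content color takes the tournament display info's PrimaryColor.
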WidgetBlueprint /CompeteOldTab/Frontend/Showdown/EventScoreBigStat
graphs [
 EventGraph {nodes [
   Event PreConstruct()
   Event Construct()
   Event Tick()
  ]}
 SetData {nodes [
   n1 = SetData!() -> (then, Label:Text, Value:int, ValueContext:int, FormatterText:Text, Tournament Display Info:Fort Tournament Display Info Structure) {is=title K2Node_FunctionEntry}
   > "SetText (Text)"(self=Get StatLabelText() -> (StatLabelText:Common Text Object Reference), InText=@n1)
   > "SetText (Text)"(self=Get StatValueText() -> (StatValueText:Common Text Object Reference), InText="To Text (Integer)"(Value=@n1.Value))
   > n2 = Branch(Condition="integer >= integer"!(A:int=@n1.ValueContext, B:int=0) -> (ReturnValue:bool) {is=title K2Node_CallFunction})
   n3 = Get RichTextBlock_GoalScore() -> (RichTextBlock_GoalScore:Rich Text Block Object Reference)
   > SetText(self=@n3, InText=Format Text!(Format=@n1.FormatterText, 0:Text="To Text (Integer)"(Value=@n1.ValueContext)) -> (Result))
   > SetVisibility(
    self=@n3
    InVisibility=Select(Option 0=Collapsed, Option 1=SelfHitTestInvisible, Index="integer > integer"(A=@n1.ValueContext, B=0) {is=title K2Node_CallFunction})
   )
   > SetContentColorAndOpacity!(
    execute
    self:Border Object Reference=Get CommonBorder_GoalScore() -> (CommonBorder_GoalScore:Common Border Object Reference)
    InContentColorAndOpacity:Linear Color Structure=Break Fort Tournament Display Info!(FortTournamentDisplayInfo:"Fort Tournament Display Info Structure (by ref)"=@n1.Tournament Display Info) -> (
     TitleLine1:Text
     TitleLine2:Text
     ScheduleInfo:Text
     FlavorDescription:Text
     DetailsDescription:Text
     ShortFormatTitle:Text
     LongFormatTitle:Text
     BackgroundTitle:Text
     PinScoreRequirement:int
     PinEarnedText:Text
     BaseColor:Linear Color Structure
     PrimaryColor:Linear Color Structure
     SecondaryColor:Linear Color Structure
     HighlightColor:Linear Color Structure
     TitleColor:Linear Color Structure
     ShadowColor:Linear Color Structure
     BackgroundLeftColor:Linear Color Structure
     BackgroundRightColor:Linear Color Structure
     BackgroundTextColor:Linear Color Structure
     PosterFadeColor:Linear Color Structure
     AlertText:Text
     SeriesPointLeaderboardName:Text
     AlertType:EFortTournamentAlertType Enum
     RoundDisplayNames:Array of Texts
    ) {is=title K2Node_BreakStruct}.PrimaryColor
   ) -> (then)
   SetVisibility(execute=@n2.else, self=Get RichTextBlock_GoalScore(), InVisibility=Hidden)
  ]}
]
variables []
widgets {
 info {inheritedWidgetCount=0, parentClass=&UserWidget, rootWidgetClass=&Overlay, widgetCount=8}
 widgets [
  Overlay_0 {widgetClassPath=&Overlay, widget=Overlay_0}
  HorizontalBox_2 {
   widgetClassPath=&HorizontalBox
   parent=Overlay_0
   widget=HorizontalBox_2
   slot {class=OverlaySlot, properties {horizontalAlignment=HAlign_Center, verticalAlignment=VAlign_Center}}
  }
  SizeBox_1 {
   widgetClassPath=&SizeBox
   parent=HorizontalBox_2
   widget=SizeBox_1
   properties {bOverride_WidthOverride=true, widthOverride=160}
   slot {class=HorizontalBoxSlot, properties {verticalAlignment=VAlign_Center, padding=<ltrb 0 0 30 0>}}
  }
  ScaleBox_0 {
   widgetClassPath=&ScaleBox
   parent=SizeBox_1
   widget=ScaleBox_0
   slot {class=SizeBoxSlot, properties {horizontalAlignment=HAlign_Left}}
  }
  StatLabelText {
   widgetClassPath=&CommonTextBlock
   bIsVariable=true
   parent=ScaleBox_0
   widget=StatLabelText
   properties {
    justification=Right
    lineHeightPercentage=0.7
    text="MATCHES\r\nPLAYED"
    font {
     bForceMonospaced=false
     bMaterialIsStencil=false
     fontMaterial=null
     fontObject=&Burbank
     letterSpacing=0
     monospacedWidth=1
     size=24
     skewAmount=0
     typefaceFontName=Burbank-PlayerPower
     outlineSettings {
      bApplyOutlineToDropShadows=false
      bMiteredCorners=false
      bSeparateFillAlpha=false
      outlineMaterial=null
      outlineSize=0
      outlineColor=<rgba 0 0 0 1>
     }
    }
    renderTransform {angle=0, scale=<1 1>, shear=<-7 0>, translation=<0 0>}
   }
   slot {class=ScaleBoxSlot}
  }
  StatValueText {
   widgetClassPath=&CommonTextBlock
   bIsVariable=true
   parent=HorizontalBox_2
   widget=StatValueText
   properties {
    justification=Right
    text=0
    font {
     bForceMonospaced=false
     bMaterialIsStencil=false
     fontMaterial=null
     fontObject=&Burbank
     letterSpacing=0
     monospacedWidth=1
     size=70
     skewAmount=0
     typefaceFontName=Burbank-Black
     outlineSettings {
      bApplyOutlineToDropShadows=false
      bMiteredCorners=false
      bSeparateFillAlpha=false
      outlineMaterial=null
      outlineSize=0
      outlineColor=<rgba 0 0 0 1>
     }
    }
   }
   slot {class=HorizontalBoxSlot, properties {verticalAlignment=VAlign_Center}}
  }
  CommonBorder_GoalScore {
   widgetClassPath=&CommonBorder
   bIsVariable=true
   parent=HorizontalBox_2
   widget=CommonBorder_GoalScore
   properties {contentColorAndOpacity=<rgba 0.0185 1 0.9047 1>, padding=<ltrb 0 0 0 0>}
   slot {class=HorizontalBoxSlot, properties {verticalAlignment=VAlign_Bottom, padding=<ltrb 10 0 0 18>}}
  }
  RichTextBlock_GoalScore {
   widgetClassPath=/Script/UMG.RichTextBlock
   bIsVariable=true
   parent=CommonBorder_GoalScore
   widget=RichTextBlock_GoalScore
   properties {
    text="<WindowDetailsPointTotal>/ 0</> <WindowDetailsPointTotal_Small>pts</>"
    textStyleSet=&/TournamentStyles
    visibility=Hidden
   }
   slot {class=BorderSlot, properties {padding=<ltrb 0 0 0 0>}}
  }
 ]
}
