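# EventsSimpleButton is a CommonButtonLegacy widget: a skewed, shadowed button with a text label for the Showdown frontend.
# PreConstruct, Construct and UpdateColor apply label, padding and a named color map to dynamic font and image materials.
# InputTypeChanged collapses the background on gamepad and recolors the text with the button color.
# Hover animates the material Hover parameters, and Pressed plays a scale dip through click and mouse events.
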WidgetBlueprint /CompeteOldTab/Frontend/Showdown/EventsSimpleButton
animations [
 Hover {
  binding_count=2
  display_label=""
  length_seconds=0.3493
  object=&self:Hover
  sequence {bindings [
    CenterButtonTextWidget {
     id=#1
     possessed_class=CommonTextBlock
     tracks [{
       type=MovieSceneWidgetMaterialTrack
       display_name=Font.FontMaterial
       sections [{type=MovieSceneParameterSection, end_frame=6, start_frame=0, channels [Hover {type=MovieSceneScriptingFloatChannel, keys [[0,0],[6.986,1]]}]}]
      }]
    }
    Image_ButtonTop {
     id=#2
     possessed_class=Image
     tracks [{
       type=MovieSceneWidgetMaterialTrack
       sections [{type=MovieSceneParameterSection, end_frame=6, start_frame=0, channels [Hover {type=MovieSceneScriptingFloatChannel, keys [[0,0],[6.986,1]]}]}]
      }]
    }
   ]}
 }
 Pressed {
  binding_count=1
  display_label=""
  length_seconds=0.15
  object=&self:Pressed
  sequence {bindings [Overlay_Button {
     id=#3
     possessed_class=Overlay
     tracks [{
       type=MovieScene2DTransformTrack
       sections [{
         type=MovieScene2DTransformSection
         end_frame=3
         start_frame=0
         channels [
          Translation.X {type=MovieSceneScriptingFloatChannel, const [0,1]}
          Translation.Y {type=MovieSceneScriptingFloatChannel, const [0,1]}
          Angle {type=MovieSceneScriptingFloatChannel, const [0,1]}
          Scale.X {type=MovieSceneScriptingFloatChannel, keys [[0,1],[1,0.8],[3,1]]}
          Scale.Y {type=MovieSceneScriptingFloatChannel, keys [[0,1],[1,0.8],[3,1]]}
          Shear.X {type=MovieSceneScriptingFloatChannel, const [0,1]}
          Shear.Y {type=MovieSceneScriptingFloatChannel, const [0,1]}
         ]
        }]
      }]
    }]}
 }
]
graphs [
 EventGraph {nodes [
   InputTypeChanged!() -> (OutputDelegate, then, bUsingGamepad:bool=~) {is=title K2Node_CustomEvent}
   n1 = IsUsingGamepad!(self:Fort Global UIContext Object Reference=Get FortGlobalUIContext!() -> (ReturnValue:Fort Global UIContext Object Reference) {is=title K2Node_GetBlueprintContextEx}) -> (ReturnValue:bool)
   > SetVisibility(
    self=Get Overlay_ButtonBG() -> (Overlay_ButtonBG:Overlay Object Reference)
    InVisibility=Select(Option 0=SelfHitTestInvisible, Option 1=Collapsed, Index=@n1)
   )
   > n2 = GetDynamicFontMaterial!(execute, self:Text Object Reference=Get CenterButtonTextWidget() -> (CenterButtonTextWidget:Common Text Object Reference)) -> (then, ReturnValue:Material Instance Dynamic Object Reference)
   > SetVectorParameterValue(
    self=@n2
    ParameterName=BaseColor
    Value=Select(
     Option 0:Linear Color Structure=Get DefaultTextColor() -> (DefaultTextColor:Linear Color Structure)
     Option 1:Linear Color Structure=Get GamepadButtonTextColor() -> (GamepadButtonTextColor:Linear Color Structure)
     Index=@n1
    ) -> (ReturnValue:Linear Color Structure)
   )
   n3 = UpdateColor!() -> (OutputDelegate, then, ParameterName:String, New Color:Linear Color Structure) {is=title K2Node_CustomEvent}
   > n4 = Add!(
    execute
    TargetMap:Map of Strings to Linear Color Structures=Get Color() -> (Color:Map of Strings to Linear Color Structures)
    Key:"String (by ref)"=@n3.ParameterName
    Value:"Linear Color Structure (by ref)"=@n3.New Color
   ) -> (then) {is=title K2Node_CallFunction}
   n5 = Get ButtonNameText() -> (ButtonNameText:Text)
   n6 = Get CenterButtonTextWidget()
   "SetText (Text)"(
    execute=[@n4.then, @n15.then, @n16.then]
    self=@n6
    InText=Select(Option 0:Text=@n5, Option 1:Text=TextToUpper!(InText:"Text (by ref)"=@n5) -> (ReturnValue:Text), Index=Get toUpper() -> (toUpper:bool)) -> (ReturnValue:Text)
   )
   > n7 = GetDynamicFontMaterial(self=@n6)
   n8 = Get Color()
   n9 = Find(TargetMap:Map of Strings to Linear Color Structures=@n8, Key=TextColor) -> (Value:Linear Color Structure)
   > SetVectorParameterValue(self=@n7, ParameterName=BaseColor, Value=@n9)
   > Set DefaultTextColor(DefaultTextColor:Linear Color Structure=@n9) -> (Output_Get:Linear Color Structure)
   > SetVectorParameterValue(self=@n7, ParameterName=HoverColor, Value=Find(TargetMap=@n8, Key=TextHover))
   > SetVectorParameterValue(self=@n7, ParameterName=PressedColor, Value=Find(TargetMap=@n8, Key=TextPressed))
   > SetVectorParameterValue(self=@n7, ParameterName=PressedHoverColor, Value=Find(TargetMap=@n8, Key=TextPressedHover))
   > n10 = GetDynamicMaterial(self=Get Image_ButtonTop() -> (Image_ButtonTop:Image Object Reference))
   n11 = Get Color()
   n12 = Find(TargetMap=@n11, Key=ButtonColor1)
   > SetVectorParameterValue(self=@n10, ParameterName=Color1, Value=@n12)
   > Set GamepadButtonTextColor(GamepadButtonTextColor:Linear Color Structure=@n12) -> (Output_Get:Linear Color Structure)
   > SetVectorParameterValue(self=@n10, ParameterName=Color2, Value=Find(TargetMap=@n11, Key=ButtonColor2))
   > SetVectorParameterValue(self=@n10, ParameterName=Hover, Value=Find(TargetMap=@n11, Key=ButtonHover))
   > SetBrushFromMaterial(self=Get Image_ButtonTop(), Material=@n10)
   n13 = BreakColor(InColor=@n12)
   > SetVisibility(
    self=Get Image_Shadow() -> (Image_Shadow:Image Object Reference)
    InVisibility=Select(
     Option 0:ESlateVisibility Enum=HitTestInvisible
     Option 1:ESlateVisibility Enum=Collapsed
     Index="float <= float"!(A:double=@n13.A, B:double=0) -> (ReturnValue:bool) {is=title K2Node_CallFunction}
    ) -> (ReturnValue:ESlateVisibility Enum)
   )
   n14 = BreakVector2D(InVec=Get ButtonPadding() -> (ButtonPadding:Vector 2D Structure))
   > SetPadding!(
    execute
    self:Border Object Reference=Get CommonBorder_VariablePadding() -> (CommonBorder_VariablePadding:Common Border Object Reference)
    InPadding:Margin Structure=Make Margin(Left=@n14, Top=@n14.Y, Right=@n14, Bottom=@n14.Y)
   ) -> (then)
   n15 = Event PreConstruct()
   n16 = Event Construct()
   Event On Hovered()
   n17 = Self-Reference()
   n18 = Get Hover() -> (Hover:Widget Animation Object Reference)
   > Anim Transition!(Exec:Exec, In:bool=true, Animation:Widget Animation Object Reference=@n18, PlaySpeed:double=3, Widget:User Widget Object Reference=@n17) -> (ExecOut:Exec) {is=title K2Node_MacroInstance}
   Event On Unhovered()
   > Anim Transition(Animation=@n18, PlaySpeed=1, Widget=@n17)
   Event On Clicked()
   > PlayAnimation(InAnimation=Get Pressed() -> (Pressed:Widget Animation Object Reference))
   SetVisibility(InVisibility=Select(
     Option 0=Collapsed
     Option 1=HitTestInvisible
     Index="On Input Method Changed (InputActionWidget)"!() -> (OutputDelegate, then, bUsingGamepad:bool) {is=title K2Node_ComponentBoundEvent}.bUsingGamepad
    ))
  ]}
 OnMouseButtonDown {nodes [
   OnMouseButtonDown!() -> (then, MyGeometry:Geometry Structure, MouseEvent:"Pointer Event Structure (by ref)") {is=title K2Node_FunctionEntry}
   > Anim Transition(In=true, Animation=Get Pressed(), PlaySpeed=2, Widget=Self-Reference())
   > Return Node!(execute, ReturnValue:Event Reply Structure=Unhandled())
  ]}
 OnMouseButtonUp {nodes [
   OnMouseButtonUp!() -> (then, MyGeometry:Geometry Structure, MouseEvent:"Pointer Event Structure (by ref)") {is=title K2Node_FunctionEntry}
   > Anim Transition(Animation=Get Pressed(), PlaySpeed=1, Widget=Self-Reference())
   > Return Node(ReturnValue=Unhandled())
  ]}
]
variables [
 ButtonNameText {type=string, default="[ph] Button Text"}
 ButtonPadding {type=Vector2D, default=<20 8>}
 Color {
  type="map<LinearColor>"
  default {
   ButtonColor1=<rgba 0 0.3289 1 1>
   ButtonColor2=<rgba 0 0.3289 1 1>
   ButtonHover=<rgba 1 1 1 1>
   TextColor=<rgba 1 1 1 1>
   TextHover=<rgba 0 0 0 1>
   TextPressed=<rgba 0 0 0 1>
   TextPressedHover=<rgba 0 0 0 1>
  }
 }
 DefaultTextColor {type=LinearColor, default=<rgba 0 0 0 0>}
 GamepadButtonTextColor {type=LinearColor, default=<rgba 0 0 0 0>}
 toUpper {type=boolean, default=true}
 XL {type=boolean, default=true}
]
widgets {
 info {inheritedWidgetCount=1, parentClass=&CommonButtonLegacy, rootWidgetClass=&Overlay, widgetCount=10}
 widgets [
  Overlay_1 {widgetClassPath=&Overlay, widget=Overlay_1}
  Overlay_Button {
   widgetClassPath=&Overlay
   parent=Overlay_1
   widget=Overlay_Button
   slot {class=OverlaySlot, properties {verticalAlignment=VAlign_Center}}
  }
  SizeBox_0 {
   widgetClassPath=&SizeBox
   parent=Overlay_Button
   widget=SizeBox_0
   properties {bOverride_HeightOverride=true, bOverride_WidthOverride=true}
   slot {class=OverlaySlot, properties {horizontalAlignment=HAlign_Fill, verticalAlignment=VAlign_Fill}}
  }
  Overlay_ButtonBG {widgetClassPath=&Overlay, bIsVariable=true, parent=SizeBox_0, widget=Overlay_ButtonBG, slot {class=SizeBoxSlot}}
  Image_Shadow {
   widgetClassPath=&Image
   bIsVariable=true
   parent=Overlay_ButtonBG
   widget=Image_Shadow
   properties {
    brush {
     bIsDynamicallyLoaded=false
     drawAs=Image
     imageType=FullColor
     mirroring=NoMirror
     resourceName=None
     resourceObject=/CompeteOldTab/Frontend/Showdown/Materials/M_UI_SkewBox_SButton
     tiling=NoTile
     imageSize=<32 32>
     margin=<ltrb 0.01 0.01 0.01 0.01>
     outlineSettings {
      bUseBrushTransparency=false
      roundingType=HalfHeightRadius
      width=0
      color {colorUseRule=UseColor_Specified, specifiedColor=<rgba 0 0 0 0>}
      cornerRadii=<0 0 0 0>
     }
     tintColor {colorUseRule=UseColor_Specified, specifiedColor=<rgba 1 1 1 1>}
     uVRegion {bIsValid=false, max=<0 0>, min=<0 0>}
    }
    colorAndOpacity=<rgba 0.0075 0.0091 0.0242 0.28>
    renderTransform {angle=0, scale=<1 1>, shear=<0 0>, translation=<7 7>}
   }
   slot {class=OverlaySlot, properties {horizontalAlignment=HAlign_Fill, verticalAlignment=VAlign_Fill}}
  }
  Image_ButtonTop {
   widgetClassPath=&Image
   bIsVariable=true
   parent=Overlay_ButtonBG
   widget=Image_ButtonTop
   properties {brush {
     bIsDynamicallyLoaded=false
     drawAs=Image
     imageType=NoImage
     mirroring=NoMirror
     resourceName=None
     resourceObject=&M_UI_SkewBox_SButton
     tiling=NoTile
     imageSize=<256 256>
     margin=<ltrb 0.2 0.2 0.2 0.2>
     outlineSettings {
      bUseBrushTransparency=false
      roundingType=HalfHeightRadius
      width=0
      color {colorUseRule=UseColor_Specified, specifiedColor=<rgba 0 0 0 0>}
      cornerRadii=<0 0 0 0>
     }
     tintColor {colorUseRule=UseColor_Specified, specifiedColor=<rgba 1 1 1 1>}
     uVRegion {bIsValid=false, max=<0 0>, min=<0 0>}
    }}
   slot {class=OverlaySlot, properties {horizontalAlignment=HAlign_Fill, verticalAlignment=VAlign_Fill}}
  }
  CommonBorder_VariablePadding {
   widgetClassPath=&CommonBorder
   bIsVariable=true
   parent=Overlay_Button
   widget=CommonBorder_VariablePadding
   properties {padding=<ltrb 20 6 20 6>}
   slot {class=OverlaySlot, properties {horizontalAlignment=HAlign_Fill, verticalAlignment=VAlign_Fill}}
  }
  HorizontalBox_0 {
   widgetClassPath=&HorizontalBox
   parent=CommonBorder_VariablePadding
   widget=HorizontalBox_0
   slot {class=BorderSlot, properties {padding=<ltrb 20 6 20 6>}}
  }
  InputActionWidget {
   widgetClassPath=&CommonActionWidget
   bInherited=true
   bIsVariable=true
   parent=HorizontalBox_0
   widget=InputActionWidget
   slot {class=HorizontalBoxSlot, properties {padding=<ltrb -5 0 10 0>}}
  }
  CenterButtonTextWidget {
   widgetClassPath=&CommonTextBlock
   bIsVariable=true
   parent=HorizontalBox_0
   widget=CenterButtonTextWidget
   properties {
    justification=Center
    lineHeightPercentage=0.8
    text=TEXT BLOCK
    font {
     bForceMonospaced=false
     bMaterialIsStencil=false
     fontMaterial=null
     fontObject=&Burbank
     letterSpacing=0
     monospacedWidth=1
     size=25
     skewAmount=0
     typefaceFontName=Burbank-PlayerPower
     outlineSettings {
      bApplyOutlineToDropShadows=false
      bMiteredCorners=false
      bSeparateFillAlpha=false
      outlineMaterial=null
      outlineSize=0
      outlineColor=<rgba 0 0 0 1>
     }
    }
   }
   slot {class=HorizontalBoxSlot}
  }
 ]
}
