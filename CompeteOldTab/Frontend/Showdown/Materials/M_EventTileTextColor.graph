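# Translucent UI material that, by its name, colors text on Showdown event tiles according to focus and ended state.
# Chained lerps blend BaseColor toward FocusColor and EndedColor, with Focus multiplied by Ended selecting a separate Ended_FocusColor when both apply.
# A parameterized alpha texture, defaulting to white, is sampled with UVs scaled about the center and then rotated by AlphaRotation.
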Material /CompeteOldTab/Frontend/Showdown/Materials/M_EventTileTextColor
graph {
 properties {blendMode=BLEND_Translucent, materialDomain=MD_UI, pixelDepthOffsetMode=PDOM_Legacy, usedShadingModels=Default Lit}
 nodes [
  n1 = VectorParameter() {parameterName=BaseColor, defaultValue=<rgba 1 1 1 1>}
  n2 = VectorParameter() {parameterName=FocusColor, defaultValue=<rgba 1 1 1 0>}
  n3 = ScalarParameter() {parameterName=Focus}
  n4 = ScalarParameter() {parameterName=Ended}
  LinearInterpolate(
   A=LinearInterpolate(
    A=LinearInterpolate(A=@n1.RGB, B=@n2.RGB, Alpha=@n3)
    B=VectorParameter() {parameterName=EndedColor, defaultValue=<rgba 1 1 1 0>}.RGB
    Alpha=@n4
   )
   B=VectorParameter() {parameterName=Ended_FocusColor, defaultValue=<rgba 1 1 1 0>}.RGB
   Alpha=Multiply(A=ScalarParameter() {parameterName=Focus}, B=@n4)
  )
  n5 = TextureCoordinate()
  n6 = AppendVector(A=ScalarParameter() {defaultValue=1, parameterName=ScaleX}, B=ScalarParameter() {defaultValue=1, parameterName=ScaleY})
  TextureSampleParameter2D(UVs=MaterialFunctionCall(
    UVs=MaterialFunctionCall(UVs=@n5, Texture Scale=@n6) {materialFunction=&ScaleUVsByCenter}.UVs
    "Rotation Angle (0-1)"=ScalarParameter() {parameterName=AlphaRotation}
   ) {materialFunction=&CustomRotator}.Rotated Values) {parameterName=AlphaTexture, texture=/Game/Global/Textures/Default/Blanks/T_WhiteUserInterface}
 ]
}
