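# Showdown tournament tile widget showing an event window's start or end time and a socket holding its pin or eligibility lock.
# Refresh Data sets labels, visibility and ended/pin flags from event state, and broadcasts state, display info and round-change delegates.
# Dynamic materials, including one font material shared by four texts, take tournament colors, with some shades derived by brightness math.
# Focus plays normal or ended animations, switching unfinished events from date to countdown.
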WidgetBlueprint /CompeteOldTab/Frontend/Showdown/ShowdownEventTileInfo
animations [
 TransitionSize {
  binding_count=3
  display_label=""
  length_seconds=0.2978
  object=&self:TransitionSize
  sequence {bindings [
    CommonBorder_Date {
     id=#1
     possessed_class=CommonBorder
     tracks [{
       type=MovieScene2DTransformTrack
       sections [{
         type=MovieScene2DTransformSection
         start_frame=0
         channels [
          Translation.X {type=MovieSceneScriptingFloatChannel}
          Translation.Y {type=MovieSceneScriptingFloatChannel, keys [[0,-128.9504],[5.9557,0]]}
          Angle {type=MovieSceneScriptingFloatChannel}
          Scale.X {type=MovieSceneScriptingFloatChannel}
          Scale.Y {type=MovieSceneScriptingFloatChannel}
          Shear.X {type=MovieSceneScriptingFloatChannel}
          Shear.Y {type=MovieSceneScriptingFloatChannel}
         ]
        }]
      }]
    }
    VerticalBoxSlot_1 {
     id=#2
     parent_id=#3
     possessed_class=VerticalBoxSlot
     tracks [{
       type=MovieSceneMarginTrack
       display_name=Padding
       sections [{
         type=MovieSceneMarginSection
         start_frame=0
         channels [
          Left {type=MovieSceneScriptingFloatChannel}
          Top {type=MovieSceneScriptingFloatChannel}
          Right {type=MovieSceneScriptingFloatChannel, keys [[0,-12],[5.9557,0]]}
          Bottom {type=MovieSceneScriptingFloatChannel}
         ]
        }]
      }]
    }
    TimeDescriptionText {id=#3, display_name=ScaleBoxTimeFormat, possessed_class=CommonTextBlock, tracks []}
   ]}
 }
 Focus_Ended {
  binding_count=3
  display_label=""
  length_seconds=0.25
  object=&self:Focus_Ended
  sequence {bindings [
    TimeDescriptionText {
     id=#4
     possessed_class=CommonTextBlock
     tracks [{
       type=MovieSceneWidgetMaterialTrack
       display_name=Font.FontMaterial
       sections [{
         type=MovieSceneParameterSection
         end_frame=5
         start_frame=0
         channels [Focus {type=MovieSceneScriptingFloatChannel, keys [[0,0],[4.9997,1]]}]
        }]
      }]
    }
    CommonBorder_Date {id=#5, possessed_class=CommonBorder, tracks []}
    Image_Lock {
     id=#6
     possessed_class=Image
     tracks [{
       type=MovieSceneWidgetMaterialTrack
       sections [{
         type=MovieSceneParameterSection
         end_frame=5
         start_frame=0
         channels [Focus {type=MovieSceneScriptingFloatChannel, keys [[0,0],[4.9997,1]]}]
        }]
      }]
    }
   ]}
 }
 Focus {
  binding_count=4
  display_label=""
  length_seconds=0.25
  object=&self:Focus
  sequence {bindings [
    CommonBorder_DateFocus {
     id=#7
     display_name=CommonBorder_0
     possessed_class=CommonBorder
     tracks [{
       type=MovieScene2DTransformTrack
       sections [{
         type=MovieScene2DTransformSection
         end_frame=5
         start_frame=0
         channels [
          Translation.X {type=MovieSceneScriptingFloatChannel, keys [[0,0],[4.9997,15]]}
          Translation.Y {type=MovieSceneScriptingFloatChannel, keys [[0,0],[4.9997,15]]}
          Angle {type=MovieSceneScriptingFloatChannel}
          Scale.X {type=MovieSceneScriptingFloatChannel}
          Scale.Y {type=MovieSceneScriptingFloatChannel}
          Shear.X {type=MovieSceneScriptingFloatChannel}
          Shear.Y {type=MovieSceneScriptingFloatChannel}
         ]
        }]
      }]
    }
    TimeDescriptionText {
     id=#8
     possessed_class=CommonTextBlock
     tracks [{
       type=MovieSceneWidgetMaterialTrack
       display_name=Font.FontMaterial
       sections [{
         type=MovieSceneParameterSection
         end_frame=5
         start_frame=0
         channels [Focus {type=MovieSceneScriptingFloatChannel, keys [[0,0],[4.9997,1]]}]
        }]
      }]
    }
    Image_Lock {
     id=#9
     possessed_class=Image
     tracks [{
       type=MovieSceneWidgetMaterialTrack
       sections [{
         type=MovieSceneParameterSection
         end_frame=5
         start_frame=0
         channels [Focus {type=MovieSceneScriptingFloatChannel, keys [[0,0],[4.9997,1]]}]
        }]
      }]
    }
    CommonBorder_Date {id=#5, possessed_class=CommonBorder, tracks []}
   ]}
 }
 HideSocketShadow {
  binding_count=4
  display_label=""
  length_seconds=0.3
  object=&self:HideSocketShadow
  sequence {bindings [
    Overlay_Pin {
     id=#10
     possessed_class=Overlay
     tracks [{
       type=MovieScene2DTransformTrack
       sections [{
         type=MovieScene2DTransformSection
         end_frame=6
         start_frame=0
         channels [
          Translation.X {type=MovieSceneScriptingFloatChannel}
          Translation.Y {type=MovieSceneScriptingFloatChannel}
          Angle {type=MovieSceneScriptingFloatChannel}
          Scale.X {type=MovieSceneScriptingFloatChannel, keys [[0,1],[5.9997,1.2]]}
          Scale.Y {type=MovieSceneScriptingFloatChannel, keys [[0,1],[5.9997,1.2]]}
          Shear.X {type=MovieSceneScriptingFloatChannel}
          Shear.Y {type=MovieSceneScriptingFloatChannel}
         ]
        }]
      }]
    }
    Image_Socket {
     id=#11
     possessed_class=Image
     tracks [{
       type=MovieSceneWidgetMaterialTrack
       sections [{
         type=MovieSceneParameterSection
         end_frame=6
         start_frame=0
         channels [
          Focus {type=MovieSceneScriptingFloatChannel, keys [[0,0],[5.9997,1]]}
          ShadowFadeR {type=MovieSceneScriptingFloatChannel}
          ShadowFadeG {type=MovieSceneScriptingFloatChannel}
          ShadowFadeB {type=MovieSceneScriptingFloatChannel}
          ShadowFadeA {type=MovieSceneScriptingFloatChannel, keys [[0,0],[5.9987,1]]}
         ]
        }]
      }]
    }
    Image_Lock {
     id=#12
     possessed_class=Image
     tracks [{
       type=MovieSceneWidgetMaterialTrack
       sections [{
         type=MovieSceneParameterSection
         end_frame=6
         start_frame=0
         channels [Focus {type=MovieSceneScriptingFloatChannel, keys [[0,0],[5.9997,1]]}]
        }]
      }]
    }
    CommonBorder_Score {id=#13, possessed_class=CommonBorder, tracks []}
   ]}
 }
 HideSocketShadow_Past {
  binding_count=4
  display_label=""
  length_seconds=0.3
  object=&self:HideSocketShadow_Past
  sequence {bindings [
    Image_Lock {
     id=#14
     possessed_class=Image
     tracks [{
       type=MovieSceneWidgetMaterialTrack
       sections [{
         type=MovieSceneParameterSection
         end_frame=6
         start_frame=0
         channels [Focus {type=MovieSceneScriptingFloatChannel, keys [[0,0],[5.9997,1]]}]
        }]
      }]
    }
    Overlay_Pin {
     id=#10
     possessed_class=Overlay
     tracks [{
       type=MovieScene2DTransformTrack
       sections [{
         type=MovieScene2DTransformSection
         end_frame=6
         start_frame=0
         channels [
          Translation.X {type=MovieSceneScriptingFloatChannel}
          Translation.Y {type=MovieSceneScriptingFloatChannel}
          Angle {type=MovieSceneScriptingFloatChannel}
          Scale.X {type=MovieSceneScriptingFloatChannel, keys [[0,1],[5.9997,1.2]]}
          Scale.Y {type=MovieSceneScriptingFloatChannel, keys [[0,1],[5.9997,1.2]]}
          Shear.X {type=MovieSceneScriptingFloatChannel}
          Shear.Y {type=MovieSceneScriptingFloatChannel}
         ]
        }]
      }]
    }
    Image_Socket {
     id=#11
     possessed_class=Image
     tracks [{
       type=MovieSceneWidgetMaterialTrack
       sections [{
         type=MovieSceneParameterSection
         end_frame=6
         start_frame=0
         channels [
          Focus {type=MovieSceneScriptingFloatChannel, keys [[0,0],[5.9997,1]]}
          ShadowFadeR {type=MovieSceneScriptingFloatChannel}
          ShadowFadeG {type=MovieSceneScriptingFloatChannel}
          ShadowFadeB {type=MovieSceneScriptingFloatChannel}
          ShadowFadeA {type=MovieSceneScriptingFloatChannel, thinned_from=3, keys [[0,0],[5.9997,1]]}
         ]
        }]
      }]
    }
    CommonBorder_Score {id=#13, possessed_class=CommonBorder, tracks []}
   ]}
 }
 FocusPin {
  binding_count=1
  display_label=""
  length_seconds=0.218
  object=&self:FocusPin
  sequence {bindings [OverlayPointSocket {
     id=#15
     possessed_class=Overlay
     tracks [{
       type=MovieScene2DTransformTrack
       sections [{
         type=MovieScene2DTransformSection
         end_frame=4
         start_frame=0
         channels [
          Translation.X {type=MovieSceneScriptingFloatChannel}
          Translation.Y {type=MovieSceneScriptingFloatChannel, keys [[0,0],[4.36,-19.3052]]}
          Angle {type=MovieSceneScriptingFloatChannel}
          Scale.X {type=MovieSceneScriptingFloatChannel}
          Scale.Y {type=MovieSceneScriptingFloatChannel}
          Shear.X {type=MovieSceneScriptingFloatChannel}
          Shear.Y {type=MovieSceneScriptingFloatChannel}
         ]
        }]
      }]
    }]}
 }
]
graphs [
 EventGraph {nodes [
   n1 = EventAnimFocus!() -> (OutputDelegate, then, In:bool=~) {is=title K2Node_CustomEvent}
   > n2 = Branch(Condition=Get isEnded() -> (isEnded:bool))
   n3 = Branch(execute=[@n13.then, @n2.then], Condition=@n1.In)
   n4 = Get isEnded()
   n5 = Select(
    Option 0:Widget Animation Object Reference=Get Focus() -> (Focus:Widget Animation Object Reference)
    Option 1:Widget Animation Object Reference=Get Focus_Ended() -> (Focus_Ended:Widget Animation Object Reference)
    Index=@n4
   ) -> (ReturnValue:Widget Animation Object Reference)
   n6 = Self-Reference()
   > Anim Transition!(Exec:Exec, In:bool=true, Animation:Widget Animation Object Reference=@n5, PlaySpeed:double=1, Widget:User Widget Object Reference=@n6) -> (ExecOut:Exec) {is=title K2Node_MacroInstance}
   n7 = Select(
    Option 0=Get HideSocketShadow() -> (HideSocketShadow:Widget Animation Object Reference)
    Option 1=Get HideSocketShadow_Past() -> (HideSocketShadow_Past:Widget Animation Object Reference)
    Index=@n4
   )
   n8 = Self-Reference()
   > Anim Transition(In=true, Animation=@n7, PlaySpeed=1, Widget=@n8)
   n9 = Get ShowdownPin() -> (ShowdownPin:Showdown Pin Object Reference)
   n10 = Get Focus!(self:Showdown Pin Object Reference=@n9) -> (Focus)
   > Anim Transition(In=true, Animation=@n10, PlaySpeed=1, Widget=@n9)
   n11 = Get isPin() -> (isPin:bool)
   > Branch(Condition=@n11)
   n12 = Get FocusPin() -> (FocusPin:Widget Animation Object Reference)
   > Anim Transition(In=true, Animation=@n12, PlaySpeed=1)
   Anim Transition(Exec=@n3.else, Animation=@n5, PlaySpeed=1, Widget=@n6)
   > Anim Transition(Animation=@n7, PlaySpeed=1, Widget=@n8)
   > Anim Transition(Animation=@n10, PlaySpeed=1, Widget=@n9)
   > Branch(Condition=@n11)
   > Anim Transition(Animation=@n12, PlaySpeed=1)
   n13 = SetActiveWidgetIndex(
    execute=@n2.else
    self=[
     Get CommonWidgetSwitcher_TimeFormat() -> (CommonWidgetSwitcher_TimeFormat:Common Widget Switcher Legacy Object Reference)
     Get CommonWidgetSwitcher_TimeFormatTitle() -> (CommonWidgetSwitcher_TimeFormatTitle:Common Widget Switcher Legacy Object Reference)
    ]
    Index=Select(Option 0:int=0, Option 1:int=1, Index=@n1.In) -> (ReturnValue:int)
   )
   EventStyleSocket() {is=title K2Node_CustomEvent}
   > Branch(Condition=Get isEnded())
   > n14 = GetDynamicMaterial(self=Get Image_Socket() -> (Image_Socket:Image Object Reference))
   > SetScalarParameterValue(self=@n14, ParameterName=Ended, Value=1)
   EventColorize() {is=title K2Node_CustomEvent}
   > n15 = Sequence!(execute) -> (then_0, then_1, then_2:Exec, then_3:Exec)
   n16 = Break Fort Tournament Display Info!(FortTournamentDisplayInfo:"Fort Tournament Display Info Structure (by ref)"=Get TournamentDisplayInfo() -> (TournamentDisplayInfo:Fort Tournament Display Info Structure)) -> (
    TitleLine1:Text
    TitleLine2:Text
    ScheduleInfo:Text
    FlavorDescription:Text
    DetailsDescription:Text
    ShortFormatTitle:Text
    LongFormatTitle:Text
    BackgroundTitle:Text
    PinScoreRequirement:int
    PinEarnedText:Text
    BaseColor:Linear Color Structure
    PrimaryColor:Linear Color Structure
    SecondaryColor:Linear Color Structure
    HighlightColor:Linear Color Structure
    TitleColor:Linear Color Structure
    ShadowColor:Linear Color Structure
    BackgroundLeftColor:Linear Color Structure
    BackgroundRightColor:Linear Color Structure
    BackgroundTextColor:Linear Color Structure
    PosterFadeColor:Linear Color Structure
    AlertText:Text
    SeriesPointLeaderboardName:Text
    AlertType:EFortTournamentAlertType Enum
    RoundDisplayNames:Array of Texts
   ) {is=title K2Node_BreakStruct}
   n17 = Get TextFontMaterial() -> (TextFontMaterial:Material Instance Dynamic Object Reference)
   > SetVectorParameterValue(self=@n17, ParameterName=BaseColor, Value=@n16.ShadowColor)
   > SetVectorParameterValue(self=@n17, ParameterName=FocusColor, Value=@n16.SecondaryColor)
   n18 = RGB to HSV!(InColor:Linear Color Structure=@n16.ShadowColor) -> (H:float, S:float, V:float, A:float)
   n19 = float - float(A=@n18.V, B=0.4) {is=title K2Node_CallFunction}
   n20 = "float < float"!(A:double=@n18.V, B:double=0.4) -> (ReturnValue:bool) {is=title K2Node_CallFunction}
   n21 = float + float(A=@n18.V, B=0.3) {is=title K2Node_CommutativeAssociativeBinaryOperator}
   n22 = HSV to RGB!(H:float=@n18, S:float=@n18.S, V:float="Clamp (Float)"(Value=SelectFloat(A=@n21, B=@n19, bPickA=@n20), Min=0, Max=1), A:float=@n18.A) -> (ReturnValue:Linear Color Structure)
   > SetVectorParameterValue(self=@n17, ParameterName=EndedColor, Value=@n22)
   > SetVectorParameterValue(self=@n17, ParameterName=Ended_FocusColor, Value=@n16.BackgroundRightColor)
   > n23 = GetDynamicMaterial(self=Get Image_Socket())
   n24 = BreakColor(InColor=@n16.PrimaryColor)
   n25 = MakeColor!(R:float=float + float(A=@n24, B=0.2), G:float=float + float(A=@n24.G, B=0.2), B:float=float + float(A=@n24.B, B=0.2), A:float=@n24.A) -> (ReturnValue:Linear Color Structure)
   > SetVectorParameterValue(self=@n23, ParameterName=OutlineColor, Value=@n25)
   > SetVectorParameterValue(self=@n23, ParameterName=ShapeColor, Value=@n16.PrimaryColor)
   > SetVectorParameterValue(self=@n23, ParameterName=OutlineColor_Focus, Value=@n25)
   > SetVectorParameterValue(self=@n23, ParameterName=ShapeColor_Focus, Value=@n16.SecondaryColor)
   n26 = RGB to HSV(InColor=@n16.ShadowColor)
   > SetVectorParameterValue(self=@n23, ParameterName=OutlineColor_Ended, Value=HSV to RGB(H=@n26, S=@n26.S, V=float + float(A=@n26.V, B=0.1), A=@n26.A))
   > SetVectorParameterValue(self=@n23, ParameterName=ShapeColor_Ended, Value=@n16.ShadowColor)
   n27 = Get Image_Lock() -> (Image_Lock:Image Object Reference)
   SetBrushFromMaterial(
    execute=@n15.then_2
    self=@n27
    Material=Select!(
     Unknown:Material Interface Object Reference
     Public:Material Interface Object Reference=/CompeteOldTab/Frontend/Showdown/Materials/M_EventTileTextColor_Unlock
     Private:Material Interface Object Reference=&M_EventTileTextColor_Unlock
     Locked:Material Interface Object Reference=/CompeteOldTab/Frontend/Showdown/Materials/M_EventTileTextColor_Lock
     Index:EFortEventWindowEligibility Enum=GetEventWindowEligibility!(self:Fort Showdown Detail View Object Reference) -> (ReturnValue:EFortEventWindowEligibility Enum=Unknown)
    ) -> (ReturnValue:Material Interface Object Reference)
   )
   n28 = GetDynamicMaterial(execute=@n15.then_3, self=@n27)
   > SetVectorParameterValue(self=@n28, ParameterName=BaseColor, Value=@n16.SecondaryColor)
   > SetVectorParameterValue(self=@n28, ParameterName=FocusColor, Value=@n16.PrimaryColor)
   > SetVectorParameterValue(self=@n28, ParameterName=EndedColor, Value=@n22)
   > SetVectorParameterValue(self=@n28, ParameterName=Ended_FocusColor, Value=@n16.PrimaryColor)
   Event Refresh Data() {is=title K2Node_Event}
   > n29 = Sequence!(execute) -> (then_0, then_1, then_2)
   n30 = CanGroupTournamentEventsByVisualRoundNumber!(self:Fort Showdown Detail View Object Reference) -> (ReturnValue:bool)
   n31 = IsFirstDisplayedWindow!(self:Fort Showdown Detail View Object Reference, bGroupByVisualRoundNumber:bool=@n30) -> (ReturnValue:bool)
   > Branch(Condition=AND Boolean(
     A=IsPreviousDisplayedWindowFromDifferentVisualRound!(self:Fort Showdown Detail View Object Reference, bGroupByVisualRoundNumber:bool=@n30) -> (ReturnValue:bool)
     B=NOT Boolean(A=@n31)
    ))
   > Call Multi Round Session!(execute:Exec, self:Showdown Event Tile Info Object Reference) -> (then:Exec) {is=title K2Node_CallDelegate}
   n32 = GetEventPlayerParticipationState!(self:Fort Showdown Detail View Object Reference) -> (ReturnValue:EFortShowdownEventState Enum=Unknown)
   "SetText (Text)"(
    execute=@n29.then_1
    self=Get TimeDescriptionText() -> (TimeDescriptionText:Common Text Object Reference)
    InText=Select!(
     Unknown:Text
     FutureTBD:Text="LOCTEXT(\"TBD\")"
     FutureScheduled:Text="LOCTEXT(\"STARTS\")"
     FutureNext:Text="LOCTEXT(\"STARTS\")"
     Live:Text="LOCTEXT(\"ENDS ON\")"
     LiveParticipating:Text="LOCTEXT(\"ENDS ON\")"
     LiveNotParticipating:Text="LOCTEXT(\"ENDS ON\")"
     Completed:Text="LOCTEXT(\"ENDED\")"
     CompletedParticipated:Text="LOCTEXT(\"ENDED\")"
     CompletedNotPartipated:Text="LOCTEXT(\"ENDED\")"
     Cancelled:Text="LOCTEXT(\"RAINED OUT\")"
     Index:EFortShowdownEventState Enum=@n32
    ) -> (ReturnValue:Text)
   )
   > "SetText (Text)"(
    self=Get TimeDescriptionTextRelative() -> (TimeDescriptionTextRelative:Common Text Object Reference)
    InText=Select(
     FutureScheduled="LOCTEXT(\"STARTS IN\")"
     FutureNext="LOCTEXT(\"STARTS IN\")"
     LiveParticipating="LOCTEXT(\"ENDS IN\")"
     LiveNotParticipating="LOCTEXT(\"ENDS IN\")"
     CompletedParticipated="LOCTEXT(\"ENDED\")"
     CompletedNotPartipated="LOCTEXT(\"ENDED\")"
     Cancelled="LOCTEXT(\"RAINED OUT\")"
     Index=@n32
    )
   )
   n33 = GetEventDateAsText!(
    self:Fort Showdown Detail View Object Reference
    DateTime:Date Time Structure=GetEventBeginTime!(self:Fort Showdown Detail View Object Reference) -> (ReturnValue:Date Time Structure)
    Style:EFortDateTimeStyle Enum=Short
   ) -> (ReturnValue:Text)
   n34 = GetEventDateAsText(DateTime=GetEventEndTime!(self:Fort Showdown Detail View Object Reference) -> (ReturnValue:Date Time Structure), Style=Short)
   > "SetText (Text)"(
    self=Get DateText() -> (DateText:Common Text Object Reference)
    InText=Select(
     FutureTBD="LOCTEXT(\"TBD\")"
     FutureScheduled=@n33
     FutureNext=@n33
     Live=@n34
     Completed=@n34
     Cancelled="LOCTEXT(\"RAINED OUT\")"
     Index=GetEventTimeState!(self:Fort Showdown Detail View Object Reference) -> (ReturnValue:EFortShowdownEventState Enum=Unknown)
    )
   )
   n35 = "Equal (Enum)"(A:EFortShowdownEventState Enum=@n32, B:EFortShowdownEventState Enum=LiveParticipating)
   n36 = "Equal (Enum)"(A=@n32, B=CompletedParticipated)
   > SetVisibility(
    self=Get ScaleBoxPin() -> (ScaleBoxPin:Scale Box Object Reference)
    InVisibility=Select!(Option 0:ESlateVisibility Enum=Collapsed, Option 1:ESlateVisibility Enum=HitTestInvisible, Index:bool=OR Boolean(A=@n35, B=@n36)) -> (ReturnValue:ESlateVisibility Enum)
   )
   > n37 = Set isEnded(isEnded:bool=OR Boolean!(
     A="Equal (Enum)"(A=@n32, B=CompletedParticipated)
     B="Equal (Enum)"(A=@n32, B=CompletedNotPartipated)
     C:bool="Equal (Enum)"(A=@n32, B=Cancelled)
    ) -> (ReturnValue)) -> (Output_Get:bool)
   > SetScalarParameterValue(self=Get TextFontMaterial(), ParameterName=Ended, Value=Select(Option 0:double=0, Option 1:double=1, Index=@n37) -> (ReturnValue:double))
   > Call Set Event State!(execute:Exec, self:Showdown Event Tile Info Object Reference, EventState:EFortShowdownEventState Enum=@n32) -> (then:Exec) {is=title K2Node_CallDelegate}
   > Call Set Display Info!(execute:Exec, self:Showdown Event Tile Info Object Reference, DisplayInfo:Fort Tournament Display Info Structure=Get TournamentDisplayInfo()) -> (then:Exec) {is=title K2Node_CallDelegate}
   > EventStyleSocket() {is=title K2Node_CallFunction}
   > SetTournament!(
    execute
    self:Fort Showdown Detail View Object Reference=Get ShowdownPin()
    TournamentId:String=GetTournamentSeriesId!(self:Fort Showdown Detail View Object Reference) -> (ReturnValue:String)
    EventId:String=GetEventId!(self:Fort Showdown Detail View Object Reference) -> (ReturnValue:String)
    bForceRefreshUI:bool=true
   ) -> (then)
   > Set isPin(isPin:bool="Equal (Enum)"(
     A:EFortShowdownPinState Enum=GetEventPinState!(self:Fort Showdown Detail View Object Reference) -> (ReturnValue:EFortShowdownPinState Enum=None)
     B:EFortShowdownPinState Enum=Unlocked
    )) -> (Output_Get:bool)
   > EventColorize() {is=title K2Node_CallFunction}
   n38 = GetEventWindowEligibility() -> (ReturnValue=Unknown)
   n39 = GetEventPlayerParticipationState() -> (ReturnValue=Unknown)
   n40 = OR Boolean!(
    A="Equal (Enum)"(A:EFortShowdownEventState Enum=@n39, B:EFortShowdownEventState Enum=LiveParticipating)
    B="Equal (Enum)"(A=@n39, B=CompletedParticipated)
   ) -> (ReturnValue)
   SetVisibility(
    execute=@n29.then_2
    self=Get CommonBorder_Score() -> (CommonBorder_Score:Common Border Object Reference)
    InVisibility=Select!(
     Unknown:ESlateVisibility Enum=SelfHitTestInvisible
     Public:ESlateVisibility Enum=Hidden
     Private:ESlateVisibility Enum=Select!(Option 0:ESlateVisibility Enum=HitTestInvisible, Option 1:ESlateVisibility Enum=Hidden, Index:bool=@n40) -> (ReturnValue)
     Locked:ESlateVisibility Enum=SelfHitTestInvisible
     Index:EFortEventWindowEligibility Enum=@n38
    ) -> (ReturnValue:ESlateVisibility Enum)
   )
   > PrintString!(
    execute
    InString=Append!(
     A=Enum to String!(Enumerator:EFortShowdownEventState Enum=@n39) -> (ReturnValue:String)
     B=" // "
     C:String=Enum to String(Enumerator:EFortEventWindowEligibility Enum=@n38)
    ) -> (ReturnValue)
    bPrintToScreen
    bPrintToLog
    TextColor
    Duration=10
    Key
   ) -> (then)
   Event Construct()
   n41 = Get TimeDescriptionText()
   > n42 = GetDynamicFontMaterial!(execute, self:Text Object Reference=@n41) -> (then, ReturnValue:Material Instance Dynamic Object Reference)
   > n43 = Set TextFontMaterial(TextFontMaterial:Material Instance Dynamic Object Reference=@n42) -> (Output_Get:Material Instance Dynamic Object Reference)
   n44 = Break Slate Font Info!(SlateFontInfo:"Slate Font Info Structure (by ref)"=Get Font!(self:Text Object Reference=@n41) -> (Font:Slate Font Info Structure)) -> (
    FontObject:Object Reference
    FontMaterial:Object Reference
    OutlineSettings:Font Outline Settings Structure
    TypefaceFontName:Name
    Size:float=~
    LetterSpacing:int
    SkewAmount:float=~
    bForceMonospaced:bool=~
    bMaterialIsStencil:bool=~
    MonospacedWidth:float=~
   ) {is=title K2Node_BreakStruct}
   > SetFont(
    self=[Get TimeDescriptionTextRelative(), @n41]
    InFontInfo=MakeSlateFontInfo!(
     FontObject:Object Reference=@n44
     FontMaterial:Object Reference=@n43
     OutlineSettings:Font Outline Settings Structure=@n44.OutlineSettings
     TypefaceFontName:Name=@n44.TypefaceFontName
     Size:float=@n44.Size
     LetterSpacing:int=0
     SkewAmount:float=0
     bForceMonospaced:bool=false
     bMaterialIsStencil:bool=false
     MonospacedWidth:float=1
    ) -> (ReturnValue:Slate Font Info Structure)
   )
   n45 = Get DateText()
   n46 = Break Slate Font Info(SlateFontInfo=Get Font(self=@n45)) -> (Size=~, SkewAmount=~, bForceMonospaced=~, bMaterialIsStencil=~, MonospacedWidth=~)
   > SetFont(
    self=[Get Text_TimeRemaining() -> (Text_TimeRemaining:Common Text Object Reference), @n45]
    InFontInfo=MakeSlateFontInfo(
     FontObject=@n46
     FontMaterial=@n43
     OutlineSettings=@n46.OutlineSettings
     TypefaceFontName=@n46.TypefaceFontName
     Size=@n46.Size
     LetterSpacing=0
     SkewAmount=0
     bForceMonospaced=false
     bMaterialIsStencil=false
     MonospacedWidth=1
    )
   )
   SetAutoWrapText!(execute, self:Text Object Reference=Get Text_DebugId() -> (Text_DebugId:Text Object Reference), InAutoTextWrap:bool=false) -> (then)
   On Transition Size() {is=title K2Node_CustomEvent}
   n47 = Get ShowdownPin()
   Anim Transition(
    In=EventAnimHover!() -> (OutputDelegate, then, In:bool=~) {is=title K2Node_CustomEvent}.In
    Animation=Get Hover!(self:Showdown Pin Object Reference=@n47) -> (Hover:Widget Animation Object Reference)
    PlaySpeed=1
    Widget=@n47
   )
   Get PointFontMaterial() -> (PointFontMaterial:Material Instance Dynamic Object Reference)
  ]}
 MultiRoundSession {nodes [MultiRoundSession() {is=title K2Node_FunctionEntry}]}
 SetDisplayInfo {nodes [SetDisplayInfo!() -> (then, DisplayInfo:Fort Tournament Display Info Structure) {is=title K2Node_FunctionEntry}]}
 SetEventState {nodes [SetEventState!() -> (then, EventState:EFortShowdownEventState Enum=Unknown) {is=title K2Node_FunctionEntry}]}
]
variables [
 isEnded {type=boolean}
 isPin {type=boolean}
 MultiRoundSession {}
 PointFontMaterial {type=MaterialInstanceDynamic}
 SetDisplayInfo {}
 SetEventState {}
 TextFontMaterial {type=MaterialInstanceDynamic}
]
widgets {
 info {inheritedWidgetCount=6, parentClass=/Script/FortniteUI.FortShowdownDetailView, rootWidgetClass=&SizeBox, widgetCount=31}
 widgets [
  SizeBox_0 {widgetClassPath=&SizeBox, widget=SizeBox_0, properties {heightOverride=200, widthOverride=150}}
  Overlay_0 {widgetClassPath=&Overlay, parent=SizeBox_0, widget=Overlay_0, slot {class=SizeBoxSlot}}
  CommonBorder_1 {
   widgetClassPath=&CommonBorder
   parent=Overlay_0
   widget=CommonBorder_1
   properties {padding=<ltrb 0 0 0 0>}
   slot {class=OverlaySlot, properties {horizontalAlignment=HAlign_Fill, verticalAlignment=VAlign_Fill}}
  }
  Overlay_2 {widgetClassPath=&Overlay, parent=CommonBorder_1, widget=Overlay_2, slot {class=BorderSlot, properties {padding=<ltrb 0 0 0 0>}}}
  CommonBorder_DateFocus {
   widgetClassPath=&CommonBorder
   parent=Overlay_2
   widget=CommonBorder_DateFocus
   properties {clipping=ClipToBounds, verticalAlignment=VAlign_Bottom, padding=<ltrb 20 0 15 0>}
   slot {class=OverlaySlot, properties {horizontalAlignment=HAlign_Fill, verticalAlignment=VAlign_Fill, padding=<ltrb 0 0 -10 0>}}
  }
  SizeBox_2 {
   widgetClassPath=&SizeBox
   parent=CommonBorder_DateFocus
   widget=SizeBox_2
   properties {bOverride_WidthOverride=true, maxDesiredWidth=156.627}
   slot {class=BorderSlot, properties {verticalAlignment=VAlign_Bottom, padding=<ltrb 20 0 15 0>}}
  }
  CommonBorder_Date {
   widgetClassPath=&CommonBorder
   bIsVariable=true
   parent=SizeBox_2
   widget=CommonBorder_Date
   properties {padding=<ltrb 0 0 0 0>, renderTransform {angle=0, scale=<1 1>, shear=<0 0>, translation=<0 -128.9504>}}
   slot {class=SizeBoxSlot}
  }
  VerticalBoxDate {
   widgetClassPath=&VerticalBox
   parent=CommonBorder_Date
   widget=VerticalBoxDate
   slot {class=BorderSlot, properties {padding=<ltrb 0 0 0 0>}}
  }
  ScaleBoxTimeFormat {
   widgetClassPath=&ScaleBox
   parent=VerticalBoxDate
   widget=ScaleBoxTimeFormat
   properties {stretchDirection=DownOnly}
   slot {class=VerticalBoxSlot, properties {padding=<ltrb 0 0 -12 -5>}}
  }
  CommonWidgetSwitcher_TimeFormatTitle {
   widgetClassPath=&CommonWidgetSwitcherLegacy
   bIsVariable=true
   parent=ScaleBoxTimeFormat
   widget=CommonWidgetSwitcher_TimeFormatTitle
   properties {transitionCurveType=QuadOut, transitionDuration=0.3, transitionType=Horizontal}
   slot {class=ScaleBoxSlot, properties {horizontalAlignment=HAlign_Right}}
  }
  TimeDescriptionText {
   widgetClassPath=&CommonTextBlock
   bIsVariable=true
   parent=CommonWidgetSwitcher_TimeFormatTitle
   widget=TimeDescriptionText
   properties {
    justification=Right
    text=STARTS
    font {
     bForceMonospaced=false
     bMaterialIsStencil=false
     fontMaterial=/CompeteOldTab/Frontend/Showdown/Materials/M_EventTileTextColor
     fontObject=&Burbank
     letterSpacing=0
     monospacedWidth=1
     size=15
     skewAmount=0
     typefaceFontName=BurbankRegular-Black
     outlineSettings {
      bApplyOutlineToDropShadows=false
      bMiteredCorners=false
      bSeparateFillAlpha=false
      outlineMaterial=null
      outlineSize=0
      outlineColor=<rgba 0 0 0 1>
     }
    }
   }
   slot {class=WidgetSwitcherSlot}
  }
  TimeDescriptionTextRelative {
   widgetClassPath=&CommonTextBlock
   bIsVariable=true
   parent=CommonWidgetSwitcher_TimeFormatTitle
   widget=TimeDescriptionTextRelative
   properties {
    justification=Right
    text=STARTS IN
    font {
     bForceMonospaced=false
     bMaterialIsStencil=false
     fontMaterial=&M_EventTileTextColor
     fontObject=&Burbank
     letterSpacing=0
     monospacedWidth=1
     size=15
     skewAmount=0
     typefaceFontName=BurbankRegular-Black
     outlineSettings {
      bApplyOutlineToDropShadows=false
      bMiteredCorners=false
      bSeparateFillAlpha=false
      outlineMaterial=null
      outlineSize=0
      outlineColor=<rgba 0 0 0 1>
     }
    }
   }
   slot {class=WidgetSwitcherSlot}
  }
  SizeBox_1 {
   widgetClassPath=&SizeBox
   parent=VerticalBoxDate
   widget=SizeBox_1
   properties {bOverride_MaxDesiredWidth=true, maxDesiredWidth=120}
   slot {class=VerticalBoxSlot, properties {horizontalAlignment=HAlign_Right}}
  }
  ScaleBox_0 {widgetClassPath=&ScaleBox, parent=SizeBox_1, widget=ScaleBox_0, properties {stretchDirection=DownOnly}, slot {class=SizeBoxSlot}}
  CommonWidgetSwitcher_TimeFormat {
   widgetClassPath=&CommonWidgetSwitcherLegacy
   bIsVariable=true
   parent=ScaleBox_0
   widget=CommonWidgetSwitcher_TimeFormat
   properties {transitionCurveType=QuadOut, transitionDuration=0.33, transitionType=Horizontal}
   slot {class=ScaleBoxSlot, properties {verticalAlignment=VAlign_Top}}
  }
  DateText {
   widgetClassPath=&CommonTextBlock
   bIsVariable=true
   parent=CommonWidgetSwitcher_TimeFormat
   widget=DateText
   properties {
    justification=Right
    text=11/3
    font {
     bForceMonospaced=false
     bMaterialIsStencil=false
     fontMaterial=&M_EventTileTextColor
     fontObject=&Burbank
     letterSpacing=0
     monospacedWidth=1
     size=32
     skewAmount=0
     typefaceFontName=BurbankRegular-Black
     outlineSettings {
      bApplyOutlineToDropShadows=false
      bMiteredCorners=false
      bSeparateFillAlpha=false
      outlineMaterial=null
      outlineSize=0
      outlineColor=<rgba 0 0 0 1>
     }
    }
   }
   slot {class=WidgetSwitcherSlot}
  }
  Text_TimeRemaining {
   widgetClassPath=&CommonTextBlock
   bInherited=true
   bIsVariable=true
   parent=CommonWidgetSwitcher_TimeFormat
   widget=Text_TimeRemaining
   properties {
    text=24H 20M
    font {
     bForceMonospaced=false
     bMaterialIsStencil=false
     fontMaterial=&M_EventTileTextColor
     fontObject=&Burbank
     letterSpacing=0
     monospacedWidth=1
     size=32
     skewAmount=0
     typefaceFontName=BurbankRegular-Black
     outlineSettings {
      bApplyOutlineToDropShadows=false
      bMiteredCorners=false
      bSeparateFillAlpha=false
      outlineMaterial=null
      outlineSize=0
      outlineColor=<rgba 0 0 0 1>
     }
    }
   }
   slot {class=WidgetSwitcherSlot, properties {horizontalAlignment=HAlign_Left, verticalAlignment=VAlign_Top}}
  }
  SizeBox_PingArt {
   widgetClassPath=&SizeBox
   parent=Overlay_2
   widget=SizeBox_PingArt
   properties {bOverride_HeightOverride=true, bOverride_WidthOverride=true, heightOverride=145.4075, widthOverride=187.6196}
   slot {class=OverlaySlot, properties {verticalAlignment=VAlign_Bottom, padding=<ltrb -40 -50 0 0>}}
  }
  OverlayPointSocket {widgetClassPath=&Overlay, parent=SizeBox_PingArt, widget=OverlayPointSocket, slot {class=SizeBoxSlot}}
  CommonWidgetSwitcher_EventPinState {
   widgetClassPath=&CommonWidgetSwitcherLegacy
   bIsVariable=true
   parent=OverlayPointSocket
   widget=CommonWidgetSwitcher_EventPinState
   slot {class=OverlaySlot, properties {verticalAlignment=VAlign_Bottom, padding=<ltrb 0 -40 0 0>}}
  }
  Overlay_Pin {
   widgetClassPath=&Overlay
   parent=CommonWidgetSwitcher_EventPinState
   widget=Overlay_Pin
   properties {renderTransformPivot=<0.5 0.8>}
   slot {class=WidgetSwitcherSlot, properties {horizontalAlignment=HAlign_Center, verticalAlignment=VAlign_Bottom, padding=<ltrb -15 0 -15 0>}}
  }
  Image_Socket {
   widgetClassPath=&Image
   bIsVariable=true
   parent=Overlay_Pin
   widget=Image_Socket
   properties {brush {
     bIsDynamicallyLoaded=false
     drawAs=Image
     imageType=NoImage
     mirroring=NoMirror
     resourceName=None
     resourceObject=/CompeteOldTab/Frontend/Showdown/Materials/M_UI_BadgeSocket_DF_Inst
     tiling=NoTile
     imageSize=<148 148>
     margin=<ltrb 0 0 0 0>
     outlineSettings {
      bUseBrushTransparency=false
      roundingType=HalfHeightRadius
      width=0
      color {colorUseRule=UseColor_Specified, specifiedColor=<rgba 0 0 0 0>}
      cornerRadii=<0 0 0 0>
     }
     tintColor {colorUseRule=UseColor_Specified, specifiedColor=<rgba 1 1 1 1>}
     uVRegion {bIsValid=false, max=<0 0>, min=<0 0>}
    }}
   slot {class=OverlaySlot}
  }
  ScaleBoxPin {
   widgetClassPath=&ScaleBox
   bIsVariable=true
   parent=Overlay_Pin
   widget=ScaleBoxPin
   properties {stretch=UserSpecified, userSpecifiedScale=0.559, visibility=Collapsed}
   slot {class=OverlaySlot, properties {padding=<ltrb -14 15 0 -28>}}
  }
  ShowdownPin {
   widgetClassPath=&/ShowdownPin
   bIsVariable=true
   parent=ScaleBoxPin
   widget=ShowdownPin
   properties {bUseAbbreviatedScore=true, colorSilhouette=true, dEVPinScore=7}
   slot {class=ScaleBoxSlot}
  }
  CommonBorder_Score {
   widgetClassPath=&CommonBorder
   bIsVariable=true
   parent=Overlay_Pin
   widget=CommonBorder_Score
   properties {visibility=Collapsed, padding=<ltrb 0 0 0 0>}
   slot {class=OverlaySlot, properties {horizontalAlignment=HAlign_Center, verticalAlignment=VAlign_Center}}
  }
  Image_Lock {
   widgetClassPath=&Image
   bIsVariable=true
   parent=CommonBorder_Score
   widget=Image_Lock
   properties {
    visibility=SelfHitTestInvisible
    brush {
     bIsDynamicallyLoaded=false
     drawAs=Image
     imageType=NoImage
     mirroring=NoMirror
     resourceName=None
     resourceObject=&M_EventTileTextColor_Lock
     tiling=NoTile
     imageSize=<64 64>
     margin=<ltrb 0 0 0 0>
     outlineSettings {
      bUseBrushTransparency=false
      roundingType=HalfHeightRadius
      width=0
      color {colorUseRule=UseColor_Specified, specifiedColor=<rgba 0 0 0 0>}
      cornerRadii=<0 0 0 0>
     }
     tintColor {colorUseRule=UseColor_Specified, specifiedColor=<rgba 1 1 1 1>}
     uVRegion {bIsValid=false, max=<0 0>, min=<0 0>}
    }
    renderTransform {angle=0, scale=<1 1>, shear=<0 0>, translation=<-16 0>}
   }
   slot {class=BorderSlot, properties {padding=<ltrb 0 0 0 0>}}
  }
  Text_DebugId {
   widgetClassPath=&TextBlock
   bInherited=true
   parent=Overlay_0
   widget=Text_DebugId
   properties {
    justification=Center
    text=EventId: Event EventWindowId: EventWindow
    visibility=Hidden
    wrapTextAt=130
    wrappingPolicy=AllowPerCharacterWrapping
    font {
     bForceMonospaced=false
     bMaterialIsStencil=false
     fontMaterial=null
     fontObject=&Roboto
     letterSpacing=0
     monospacedWidth=1
     size=13
     skewAmount=0
     typefaceFontName=Bold
     outlineSettings {
      bApplyOutlineToDropShadows=false
      bMiteredCorners=false
      bSeparateFillAlpha=false
      outlineMaterial=null
      outlineSize=2
      outlineColor=<rgba 0 0 0 1>
     }
    }
   }
   slot {class=OverlaySlot, properties {horizontalAlignment=HAlign_Fill}}
  }
  Image_PosterFront {widgetClassPath=/Script/EpicCMSUIFramework.EpicCMSImage, bInherited=true, widget=null}
  Image_PosterBack {widgetClassPath=&EpicCMSImage, bInherited=true, widget=null}
  Image_LoadingScreen {widgetClassPath=&EpicCMSImage, bInherited=true, widget=null}
  Image_PlaylistTile {widgetClassPath=&EpicCMSImage, bInherited=true, widget=null}
 ]
}
