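# Widget blueprint for the round break tile in the Showdown event UI, drawn as a thin slanted cyan bar.
# A fixed-height SizeBox holds a sheared Overlay containing a single untextured Image that fills vertically.
# The short TransitionSize animation straightens the shear and evens out the bar's left and right padding.
# The event graph holds only unconnected stub events, so the widget does not play the animation itself.
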WidgetBlueprint /CompeteOldTab/Frontend/Showdown/ShowdownEventTile_RoundBreak
animations [TransitionSize {
  binding_count=3
  length_seconds=0.3043
  object=&self:TransitionSize
  sequence {bindings [
    Overlay_0 {
     id=#1
     possessed_class=Overlay
     tracks [{
       type=MovieScene2DTransformTrack
       sections [{
         type=MovieScene2DTransformSection
         end_frame=6
         start_frame=0
         channels [
          Translation.X {type=MovieSceneScriptingFloatChannel}
          Translation.Y {type=MovieSceneScriptingFloatChannel}
          Angle {type=MovieSceneScriptingFloatChannel}
          Scale.X {type=MovieSceneScriptingFloatChannel}
          Scale.Y {type=MovieSceneScriptingFloatChannel}
          Shear.X {type=MovieSceneScriptingFloatChannel, keys [[0,-16],[6.0857,0]]}
          Shear.Y {type=MovieSceneScriptingFloatChannel}
         ]
        }]
      }]
    }
    OverlaySlot_0 {
     id=#2
     parent_id=#3
     possessed_class=OverlaySlot
     tracks [{
       type=MovieSceneMarginTrack
       display_name=Padding
       sections [{
         type=MovieSceneMarginSection
         end_frame=6
         start_frame=0
         channels [
          Left {type=MovieSceneScriptingFloatChannel, keys [[0,4],[6.0857,25]]}
          Top {type=MovieSceneScriptingFloatChannel}
          Right {type=MovieSceneScriptingFloatChannel, keys [[0,60],[6.0857,25]]}
          Bottom {type=MovieSceneScriptingFloatChannel}
         ]
        }]
      }]
    }
    Image_Bg {id=#3, possessed_class=Image, tracks []}
   ]}
 }]
graphs [EventGraph {nodes [
   Event PreConstruct()
   Event Construct()
   Event Tick()
  ]}]
variables []
widgets {
 info {inheritedWidgetCount=0, parentClass=&UserWidget, rootWidgetClass=&SizeBox, widgetCount=3}
 widgets [
  SizeBox_0 {widgetClassPath=&SizeBox, widget=SizeBox_0, properties {bOverride_HeightOverride=true, heightOverride=212, widthOverride=148}}
  Overlay_0 {
   widgetClassPath=&Overlay
   parent=SizeBox_0
   widget=Overlay_0
   properties {renderTransform {angle=0, scale=<1 1>, shear=<-16 0>, translation=<0 0>}}
   slot {class=SizeBoxSlot}
  }
  Image_Bg {
   widgetClassPath=&Image
   bIsVariable=true
   parent=Overlay_0
   widget=Image_Bg
   properties {
    brush {
     bIsDynamicallyLoaded=false
     drawAs=Image
     imageType=NoImage
     mirroring=NoMirror
     resourceName=None
     resourceObject=null
     tiling=NoTile
     imageSize=<4 212.5361>
     margin=<ltrb 0 0 0 0>
     outlineSettings {
      bUseBrushTransparency=false
      roundingType=HalfHeightRadius
      width=0
      color {colorUseRule=UseColor_Specified, specifiedColor=<rgba 0 0 0 0>}
      cornerRadii=<0 0 0 0>
     }
     tintColor {colorUseRule=UseColor_Specified, specifiedColor=<rgba 1 1 1 1>}
     uVRegion {bIsValid=false, max=<0 0>, min=<0 0>}
    }
    colorAndOpacity=<rgba 0.0009 0.8632 0.8388 1>
   }
   slot {class=OverlaySlot, properties {verticalAlignment=VAlign_Fill, padding=<ltrb 4 -20 60 0>}}
  }
 ]
}
