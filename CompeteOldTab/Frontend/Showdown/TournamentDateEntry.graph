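# TournamentDateEntry is a Showdown tournament UI row, derived from FortShowdownDetailView, showing either an event's date range or a title.
# PreConstruct switches mode by isTitle, sets the title and randomizes the strikethrough material's rotation.
# Refresh Data colors text from tournament display info, shows strikethrough and shadow tint for completed events, and writes start-to-end date/time, omitting a repeated end date.
# The separate time text is collapsed, though still recolored.
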WidgetBlueprint /CompeteOldTab/Frontend/Showdown/TournamentDateEntry
graphs [EventGraph {nodes [
   n1 = EventColor!() -> (OutputDelegate, then, showdownData:"Fort Tournament Display Info Structure (by ref)") {is=title K2Node_CustomEvent}
   n2 = Break Fort Tournament Display Info!(FortTournamentDisplayInfo:"Fort Tournament Display Info Structure (by ref)"=@n1.showdownData) -> (
    TitleLine1:Text
    TitleLine2:Text
    ScheduleInfo:Text
    FlavorDescription:Text
    DetailsDescription:Text
    ShortFormatTitle:Text
    LongFormatTitle:Text
    BackgroundTitle:Text
    PinScoreRequirement:int
    PinEarnedText:Text
    BaseColor:Linear Color Structure
    PrimaryColor:Linear Color Structure
    SecondaryColor:Linear Color Structure
    HighlightColor:Linear Color Structure
    TitleColor:Linear Color Structure
    ShadowColor:Linear Color Structure
    BackgroundLeftColor:Linear Color Structure
    BackgroundRightColor:Linear Color Structure
    BackgroundTextColor:Linear Color Structure
    PosterFadeColor:Linear Color Structure
    AlertText:Text
    SeriesPointLeaderboardName:Text
    AlertType:EFortTournamentAlertType Enum
    RoundDisplayNames:Array of Texts
   ) {is=title K2Node_BreakStruct}
   n3 = Make Slate Color(SpecifiedColor=@n2.BaseColor)
   > SetColorAndOpacity(
    self=[
     Get CommonTextBlock_Date() -> (CommonTextBlock_Date:Common Text Object Reference)
     Get CommonTextBlock_Title() -> (CommonTextBlock_Title:Common Text Object Reference)
    ]
    InColorAndOpacity=@n3
   )
   n4 = "Equal (Enum)"(
    A:EFortShowdownEventState Enum=GetEventTimeState!(self:Fort Showdown Detail View Object Reference) -> (ReturnValue:EFortShowdownEventState Enum=Unknown)
    B:EFortShowdownEventState Enum=Completed
   )
   > SetColorAndOpacity(
    self=Get CommonTextBlock_Time() -> (CommonTextBlock_Time:Common Text Object Reference)
    InColorAndOpacity=Select(Option 0:Slate Color Structure=Make Slate Color(SpecifiedColor=@n2.HighlightColor), Option 1:Slate Color Structure=@n3, Index=@n4) -> (ReturnValue:Slate Color Structure)
   )
   Event Refresh Data() {is=title K2Node_Event}
   > SetVisibility(
    self=Get Image_Strkethrough() -> (Image_Strkethrough:Image Object Reference)
    InVisibility=Select(Option 0:ESlateVisibility Enum=Collapsed, Option 1:ESlateVisibility Enum=HitTestInvisible, Index=@n4) -> (ReturnValue:ESlateVisibility Enum)
   )
   n5 = Get TournamentDisplayInfo() -> (TournamentDisplayInfo:Fort Tournament Display Info Structure)
   > EventColor!(execute, self, showdownData:"Fort Tournament Display Info Structure (by ref)"=@n5) -> (then) {is=title K2Node_CallFunction}
   > SetContentColorAndOpacity!(
    execute
    self:Border Object Reference=Get CommonBorder_DateContent() -> (CommonBorder_DateContent:Common Border Object Reference)
    InContentColorAndOpacity:Linear Color Structure=Select(
     Option 0:Linear Color Structure=(R=1,G=1,B=1,A=1)
     Option 1:Linear Color Structure=Break Fort Tournament Display Info(FortTournamentDisplayInfo=@n5).ShadowColor
     Index=@n4
    ) -> (ReturnValue:Linear Color Structure)
   ) -> (then)
   n6 = GetEventBeginTime!(self:Fort Showdown Detail View Object Reference) -> (ReturnValue:Date Time Structure)
   n7 = GetEventEndTime!(self:Fort Showdown Detail View Object Reference) -> (ReturnValue:Date Time Structure)
   n8 = GetEventBeginTime()
   n9 = "Equal Exactly (Text)"!(
    A:"Text (by ref)"=GetEventDateAsText!(self:Fort Showdown Detail View Object Reference, DateTime:Date Time Structure=GetEventBeginTime(), Style:EFortDateTimeStyle Enum=Short) -> (ReturnValue:Text)
    B:"Text (by ref)"=GetEventDateAsText(DateTime=GetEventEndTime(), Style=Short)
   ) -> (ReturnValue:bool)
   n10 = Format Text!(Format="LOCTEXT(\" {EndDate}\")", EndDate:Text=GetEventDateAsText(DateTime=GetEventEndTime(), Style=Short)) -> (Result)
   > "SetText (Text)"(
    self=Get CommonTextBlock_Date()
    InText=Format Text!(
     Format="LOCTEXT(\"{StartDate} {StartTime} -{EndDate} {EndTime}\")"
     StartDate:Text=GetEventDateAsText(DateTime=@n8, Style=Short)
     StartTime:Text=GetEventTimeAsText!(self:Fort Showdown Detail View Object Reference, DateTime:Date Time Structure=@n6, Style:EFortDateTimeStyle Enum=Short) -> (ReturnValue:Text)
     EndDate=Select(Option 0:Text=@n10, Option 1:Text, Index=@n9) -> (ReturnValue:Text)
     EndTime:Text=GetEventTimeAsText(DateTime=@n7, Style=Short)
    ) -> (Result)
   )
   Event PreConstruct()
   > SetActiveWidgetIndex(
    self=Get CommonWidgetSwitcher_EntryType() -> (CommonWidgetSwitcher_EntryType:Common Widget Switcher Legacy Object Reference)
    Index=Select(Option 0:int=0, Option 1:int=1, Index=Get isTitle() -> (isTitle:bool)) -> (ReturnValue:int)
   )
   > "SetText (Text)"(self=Get CommonTextBlock_Title(), InText=Get Title() -> (Title:Text))
   > n11 = GetDynamicMaterial(self=Get Image_Strkethrough())
   > SetScalarParameterValue(self=@n11, ParameterName=Rotation, Value=RandomFloatInRange(Min=0, Max=0.1))
   Event Construct()
   Event Tick()
  ]}]
variables [
 isTitle {type=boolean}
 Title {type=string, default=Round}
]
widgets {
 info {inheritedWidgetCount=6, parentClass=/Script/FortniteUI.FortShowdownDetailView, rootWidgetClass=&Overlay, widgetCount=19}
 widgets [
  Overlay_3 {widgetClassPath=&Overlay, widget=Overlay_3}
  SizeBox_2 {
   widgetClassPath=&SizeBox
   parent=Overlay_3
   widget=SizeBox_2
   slot {class=OverlaySlot, properties {horizontalAlignment=HAlign_Fill, verticalAlignment=VAlign_Fill}}
  }
  ScaleBox_0 {widgetClassPath=&ScaleBox, parent=SizeBox_2, widget=ScaleBox_0, properties {stretchDirection=DownOnly}, slot {class=SizeBoxSlot}}
  Overlay_4 {
   widgetClassPath=&Overlay
   parent=ScaleBox_0
   widget=Overlay_4
   slot {class=ScaleBoxSlot, properties {horizontalAlignment=HAlign_Fill}}
  }
  CommonWidgetSwitcher_EntryType {
   widgetClassPath=&CommonWidgetSwitcherLegacy
   bIsVariable=true
   parent=Overlay_4
   widget=CommonWidgetSwitcher_EntryType
   slot {class=OverlaySlot, properties {horizontalAlignment=HAlign_Fill, verticalAlignment=VAlign_Fill}}
  }
  CommonBorder_DateContent {
   widgetClassPath=&CommonBorder
   bIsVariable=true
   parent=CommonWidgetSwitcher_EntryType
   widget=CommonBorder_DateContent
   properties {horizontalAlignment=HAlign_Center, padding=<ltrb 0 0 0 0>}
   slot {class=WidgetSwitcherSlot}
  }
  HorizontalBox_0 {
   widgetClassPath=&HorizontalBox
   parent=CommonBorder_DateContent
   widget=HorizontalBox_0
   slot {class=BorderSlot, properties {horizontalAlignment=HAlign_Center, padding=<ltrb 0 0 0 0>}}
  }
  SizeBox_0 {
   widgetClassPath=&SizeBox
   parent=HorizontalBox_0
   widget=SizeBox_0
   properties {widthOverride=150}
   slot {class=HorizontalBoxSlot, properties {size {sizeRule=Automatic, value=0.5}}}
  }
  CommonTextBlock_Date {
   widgetClassPath=&CommonTextBlock
   bIsVariable=true
   parent=SizeBox_0
   widget=CommonTextBlock_Date
   properties {
    justification=Right
    text=11.12.2018
    font {
     bForceMonospaced=false
     bMaterialIsStencil=false
     fontMaterial=null
     fontObject=&BurbankSmall
     letterSpacing=0
     monospacedWidth=1
     size=20
     skewAmount=0
     typefaceFontName=NotoSans-Regular
     outlineSettings {
      bApplyOutlineToDropShadows=false
      bMiteredCorners=false
      bSeparateFillAlpha=false
      outlineMaterial=null
      outlineSize=0
      outlineColor=<rgba 0 0 0 1>
     }
    }
   }
   slot {class=SizeBoxSlot}
  }
  SizeBox_1 {
   widgetClassPath=&SizeBox
   parent=HorizontalBox_0
   widget=SizeBox_1
   properties {visibility=Collapsed, widthOverride=150}
   slot {class=HorizontalBoxSlot, properties {padding=<ltrb 30 0 0 0>, size {sizeRule=Automatic, value=0.5}}}
  }
  CommonTextBlock_Time {
   widgetClassPath=&CommonTextBlock
   bIsVariable=true
   parent=SizeBox_1
   widget=CommonTextBlock_Time
   properties {
    text=7:00 AM - 9:00 PM
    visibility=Collapsed
    colorAndOpacity {colorUseRule=UseColor_Specified, specifiedColor=<rgba 0.008 1 0.2384 1>}
    font {
     bForceMonospaced=false
     bMaterialIsStencil=false
     fontMaterial=null
     fontObject=&BurbankSmall
     letterSpacing=0
     monospacedWidth=1
     size=20
     skewAmount=0
     typefaceFontName=NotoSans-Regular
     outlineSettings {
      bApplyOutlineToDropShadows=false
      bMiteredCorners=false
      bSeparateFillAlpha=false
      outlineMaterial=null
      outlineSize=0
      outlineColor=<rgba 0 0 0 1>
     }
    }
   }
   slot {class=SizeBoxSlot}
  }
  CommonTextBlock_Title {
   widgetClassPath=&CommonTextBlock
   bIsVariable=true
   parent=CommonWidgetSwitcher_EntryType
   widget=CommonTextBlock_Title
   properties {
    text=ROUND 1
    font {
     bForceMonospaced=false
     bMaterialIsStencil=false
     fontMaterial=null
     fontObject=&BurbankSmall
     letterSpacing=0
     monospacedWidth=1
     size=24
     skewAmount=0
     typefaceFontName=Black
     outlineSettings {
      bApplyOutlineToDropShadows=false
      bMiteredCorners=false
      bSeparateFillAlpha=false
      outlineMaterial=null
      outlineSize=0
      outlineColor=<rgba 0 0 0 1>
     }
    }
   }
   slot {class=WidgetSwitcherSlot, properties {horizontalAlignment=HAlign_Center, verticalAlignment=VAlign_Top}}
  }
  Image_Strkethrough {
   widgetClassPath=&Image
   bIsVariable=true
   parent=Overlay_4
   widget=Image_Strkethrough
   properties {
    visibility=Collapsed
    brush {
     bIsDynamicallyLoaded=false
     drawAs=Image
     imageType=NoImage
     mirroring=NoMirror
     resourceName=None
     resourceObject=/CompeteOldTab/Frontend/Showdown/Materials/M_UI_SkewBox_Strikethrough_Inst
     tiling=NoTile
     imageSize=<32 32>
     margin=<ltrb 0 0 0 0>
     outlineSettings {
      bUseBrushTransparency=false
      roundingType=HalfHeightRadius
      width=0
      color {colorUseRule=UseColor_Specified, specifiedColor=<rgba 0 0 0 0>}
      cornerRadii=<0 0 0 0>
     }
     tintColor {colorUseRule=UseColor_Specified, specifiedColor=<rgba 1 1 1 1>}
     uVRegion {bIsValid=false, max=<0 0>, min=<0 0>}
    }
   }
   slot {class=OverlaySlot, properties {horizontalAlignment=HAlign_Fill, verticalAlignment=VAlign_Center, padding=<ltrb -5 0 -5 0>}}
  }
  Text_TimeRemaining {widgetClassPath=&CommonTextBlock, bInherited=true, widget=null}
  Text_DebugId {widgetClassPath=&TextBlock, bInherited=true, widget=null, properties=~}
  Image_PosterFront {widgetClassPath=/Script/EpicCMSUIFramework.EpicCMSImage, bInherited=true, widget=null}
  Image_PosterBack {widgetClassPath=&EpicCMSImage, bInherited=true, widget=null}
  Image_LoadingScreen {widgetClassPath=&EpicCMSImage, bInherited=true, widget=null}
  Image_PlaylistTile {widgetClassPath=&EpicCMSImage, bInherited=true, widget=null}
 ]
}
