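# Showdown tournament tile badge widget, based on FortShowdownDetailView, that labels an event as live, by round number, or by proximity (time remaining or "NEXT!").
# Construct, data refresh and RefreshViolatorData select the switcher page, with Live taking priority.
# Round text is uppercased, or blanked by OverrideDoNotShowRound; without round display, the switcher collapses unless the event is live or next.
# A Focus animation slides LiveViolator vertically into place.
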WidgetBlueprint /Game/ShowdownEventTileViolator
animations [Focus {
  binding_count=1
  length_seconds=0.3151
  object=&self:Focus
  sequence {bindings [LiveViolator {
     id=#1
     possessed_class=LiveViolator_C
     tracks [{
       type=MovieScene2DTransformTrack
       sections [{
         type=MovieScene2DTransformSection
         channels [
          Translation.X {type=MovieSceneScriptingFloatChannel}
          Translation.Y {type=MovieSceneScriptingFloatChannel, keys [[0,-40],[6.302,0]]}
          Angle {type=MovieSceneScriptingFloatChannel}
          Scale.X {type=MovieSceneScriptingFloatChannel}
          Scale.Y {type=MovieSceneScriptingFloatChannel}
          Shear.X {type=MovieSceneScriptingFloatChannel}
          Shear.Y {type=MovieSceneScriptingFloatChannel}
         ]
        }]
      }]
    }]}
 }]
graphs [EventGraph {nodes [
   n1 = RefreshViolatorData() {is=title K2Node_CustomEvent}
   n2 = DoDisplayRoundNumber!(
    self:Fort Showdown Detail View Object Reference
    bGroupByRoundNumber:bool=CanGroupTournamentEventsByVisualRoundNumber!(self:Fort Showdown Detail View Object Reference) -> (ReturnValue:bool)
   ) -> (ReturnValue:bool)
   n3 = Branch(execute=[@n1.then, @n12.then, @n13.then], Condition=@n2)
   > n4 = Branch(Condition=Get OverrideDoNotShowRound() -> (OverrideDoNotShowRound:bool))
   n5 = Get CommonTextBlock_RoundNumber() -> (CommonTextBlock_RoundNumber:Common Text Object Reference)
   > n6 = "SetText (Text)"(self=@n5)
   n7 = SetActiveWidgetIndex(
    execute=[@n6.then, @n9.then]
    self=Get CommonWidgetSwitcher_EventState() -> (CommonWidgetSwitcher_EventState:Common Widget Switcher Legacy Object Reference)
    Index=2
   )
   n8 = "Equal (Enum)"(
    A:EFortShowdownEventState Enum=GetEventTimeState!(self:Fort Showdown Detail View Object Reference) -> (ReturnValue:EFortShowdownEventState Enum=Unknown)
    B:EFortShowdownEventState Enum=Live
   )
   SetActiveWidgetIndex(
    execute=[@n7.then, @n11.then]
    self=Get CommonWidgetSwitcher_EventState()
    Index=Select(Option 0:int=Select(Option 0=0, Option 1=2, Index=@n2), Option 1:int=1, Index=@n8) -> (ReturnValue:int)
   )
   n9 = "SetText (Text)"(
    execute=@n4.else
    self=@n5
    InText=TextToUpper!(InText:"Text (by ref)"=GetWindowRoundDisplayString!(self:Fort Showdown Detail View Object Reference) -> (ReturnValue:Text)) -> (ReturnValue:Text)
   )
   n10 = "Equal Exactly (String)"(
    A=GetEventId!(self:Fort Showdown Detail View Object Reference) -> (ReturnValue:String)
    B=GetTournamentNextEvent!(self:Fort Showdown Detail View Object Reference) -> (ReturnValue:String)
   )
   n11 = SetVisibility(
    execute=@n3.else
    self=Get CommonWidgetSwitcher_EventState()
    InVisibility=Select(Option 0:ESlateVisibility Enum=Collapsed, Option 1:ESlateVisibility Enum=SelfHitTestInvisible, Index=OR Boolean(A=@n10, B=@n8)) -> (ReturnValue:ESlateVisibility Enum)
   )
   n12 = Event Refresh Data() {is=title K2Node_Event}
   n13 = Event Construct()
  ]}]
variables [OverrideDoNotShowRound {type=boolean}]
widgets {
 info {inheritedWidgetCount=6, parentClass=/Script/FortniteUI.FortShowdownDetailView, rootWidgetClass=&Overlay, widgetCount=12}
 widgets [
  Overlay_0 {widgetClassPath=&Overlay, widget=Overlay_0}
  CommonWidgetSwitcher_EventState {
   widgetClassPath=&CommonWidgetSwitcherLegacy
   bIsVariable=true
   parent=Overlay_0
   widget=CommonWidgetSwitcher_EventState
   slot {class=OverlaySlot, properties {verticalAlignment=VAlign_Bottom}}
  }
  CommonWidgetSwitcher_Proximity {
   widgetClassPath=&CommonWidgetSwitcherLegacy
   bIsVariable=true
   parent=CommonWidgetSwitcher_EventState
   widget=CommonWidgetSwitcher_Proximity
   properties {clipping=ClipToBounds, transitionDuration=0.25, transitionType=Horizontal}
   slot {class=WidgetSwitcherSlot, properties {verticalAlignment=VAlign_Bottom}}
  }
  Text_TimeRemaining {
   widgetClassPath=&CommonTextBlock
   bInherited=true
   bIsVariable=true
   parent=CommonWidgetSwitcher_Proximity
   widget=Text_TimeRemaining
   properties {
    text=IN 2 DAYS
    font {
     bForceMonospaced=false
     bMaterialIsStencil=false
     fontMaterial=null
     fontObject=&Burbank
     letterSpacing=0
     monospacedWidth=1
     size=20
     skewAmount=0
     typefaceFontName=BurbankRegular-Black
     outlineSettings {
      bApplyOutlineToDropShadows=false
      bMiteredCorners=false
      bSeparateFillAlpha=false
      outlineMaterial=null
      outlineSize=0
      outlineColor=<rgba 0 0 0 1>
     }
    }
   }
   slot {class=WidgetSwitcherSlot}
  }
  CommonTextBlock_0 {
   widgetClassPath=&CommonTextBlock
   bIsVariable=true
   parent=CommonWidgetSwitcher_Proximity
   widget=CommonTextBlock_0
   properties {
    text="NEXT!"
    font {
     bForceMonospaced=false
     bMaterialIsStencil=false
     fontMaterial=null
     fontObject=&Burbank
     letterSpacing=0
     monospacedWidth=1
     size=20
     skewAmount=0
     typefaceFontName=BurbankRegular-Black
     outlineSettings {
      bApplyOutlineToDropShadows=false
      bMiteredCorners=false
      bSeparateFillAlpha=false
      outlineMaterial=null
      outlineSize=0
      outlineColor=<rgba 0 0 0 1>
     }
    }
    renderTransform {angle=0, scale=<1 1>, shear=<-12 0>, translation=<0 0>}
   }
   slot {class=WidgetSwitcherSlot}
  }
  LiveViolator {
   widgetClassPath=/Game/UI/Frontend/Showdown/LiveViolator
   bIsVariable=true
   parent=CommonWidgetSwitcher_EventState
   widget=LiveViolator
   properties {padding=<ltrb 50 0 0 0>, renderTransform {angle=0, scale=<1 1>, shear=<0 0>, translation=<0 -40>}}
   slot {class=WidgetSwitcherSlot, properties {horizontalAlignment=HAlign_Left, verticalAlignment=VAlign_Bottom}}
  }
  CommonTextBlock_RoundNumber {
   widgetClassPath=&CommonTextBlock
   bIsVariable=true
   parent=CommonWidgetSwitcher_EventState
   widget=CommonTextBlock_RoundNumber
   properties {
    text=ROUND X
    font {
     bForceMonospaced=false
     bMaterialIsStencil=false
     fontMaterial=null
     fontObject=&Burbank
     letterSpacing=0
     monospacedWidth=1
     size=20
     skewAmount=0
     typefaceFontName=BurbankRegular-Black
     outlineSettings {
      bApplyOutlineToDropShadows=false
      bMiteredCorners=false
      bSeparateFillAlpha=false
      outlineMaterial=null
      outlineSize=0
      outlineColor=<rgba 0 0 0 1>
     }
    }
   }
   slot {class=WidgetSwitcherSlot}
  }
  Text_DebugId {widgetClassPath=&TextBlock, bInherited=true, widget=null, properties=~}
  Image_PosterFront {widgetClassPath=/Script/EpicCMSUIFramework.EpicCMSImage, bInherited=true, widget=null}
  Image_PosterBack {widgetClassPath=&EpicCMSImage, bInherited=true, widget=null}
  Image_LoadingScreen {widgetClassPath=&EpicCMSImage, bInherited=true, widget=null}
  Image_PlaylistTile {widgetClassPath=&EpicCMSImage, bInherited=true, widget=null}
 ]
}
